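# Niagara system for the golem explosion in the HeroicTale emote, with rock chunks, dust, sparks, flares, flash, glow with light, ground cracks, decals and smoke.
# Debris, dust and smoke burst at a shared System.BurstDelay, with ramping spawn rates adding chunks and dust; spawn groups switch each particle's look and motion.
# User.SpeedCurve scales delta time in several emitters, while front-end, platform and time-of-day modifiers adjust brightness, collision and glow.
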NiagaraSystem /BRCosmetics/Animation/Game/MainPlayer/Emotes/HeroicTale/FX/NS_Emote_HeroicTale_GolemExplosion
dependencies {
 usedDataInterfaces [
  &NiagaraDataInterfaceCurve
  &NiagaraDataInterfacePlatformSet
  &NiagaraDataInterfaceColorCurve
  &NiagaraDataInterfaceVector2DCurve
  &NiagaraDataInterfaceArrayColor
 ]
 usedDynamicInputs []
 usedModules [
  &self:SystemSpawnScript.SystemScriptSource.SystemScriptGraph.NiagaraNodeAssignment_0.SetVariables_#1
  &SystemState
  &self:SystemSpawnScript.SystemScriptSource.SystemScriptGraph.NiagaraNodeAssignment_1.SetVariables_#2
 ]
}
emitters {
 Chunks {
  emitterUpdateScript [
   &EmitterState(Life Cycle Mode=::NewEnumerator1, Loop Behavior=::NewEnumerator1, Loop Delay=0.05, Loop Duration=1.8, UseLoopDelay=false)
   &SpawnBurst_Instantaneous(Spawn Count=30, Spawn Time=$System.BurstDelay, Use Spawn Probability=false, UseLoopCountLimit=false)
   &SpawnRate(
    SpawnGroup=1
    SpawnRate=&FloatFromCurve(
     FloatCurve=DataInterface() {propertyValues {Curve {defaultValue=1, keys [{tangentMode=RCTM_User, time=0.0008, value=0.1149}, <at 0.2729 0.8342>, <at 0.3413 1.7364>]}}}
     CurveIndex=$Emitter.NormalizedLoopAge
     Scale Curve=44.7054
    )
    Use Spawn Probability=false
   )
  ]
  particleSpawnScript [
   /SetVariables(Particles.VisibilityTag=&SpawnGroup()) {isSetParameters=true}
   &InitializeParticle(
    Color Maximum=<rgba 0.1195 0.0395 0.0091 1>
    Color Minimum=<rgba 0.1198 0.04 0.0094 1>
    Color Mode=::NewEnumerator2
    Lifetime=&MergeHelper_FloatSwitch(0=&RandomRangeFloat(Minimum=0.1, Maximum=0.5), 1=0, 2=0, 3=0, 4=0, Index=0, Num=::NewEnumerator3)
    Mass Mode=::NewEnumerator1
    Mesh Scale Mode=::NewEnumerator3
    Mesh Uniform Scale=&FloatFromCurve(
     FloatCurve=DataInterface() {propertyValues {Curve {keys [<at 0 0.5>,<at 0.9 1>,<at 0.95 1.5>]}}}
     CurveIndex=&RandomRangeFloat(Minimum=0, Maximum=1)
     Scale Curve=1.195
    )
    Sprite Rotation Mode=::NewEnumerator2
    Sprite UV Mode=::NewEnumerator3
    UsePositionOffset=false
    Write Material Random=true
   )
   &ShapeLocation(
    MaskSpawnGroup=false
    Non Uniform Scale=<1 1 0.5>
    Offset=<0 0 20>
    Offset Mode=::NewEnumerator0
    Override Seed=false
    Sphere Radius=&FloatFromCurve(
     FloatCurve=DataInterface() {propertyValues {Curve {defaultValue=1, keys [{tangentMode=RCTM_User, time=0, value=0.3263}, {tangentMode=RCTM_User, time=0.8823, value=1}]}}}
     CurveIndex=$System.Age
     Scale Curve=75.2447
    )
    UseOwnerScale=true
   )
   &AddVelocity(
    Module.Use Falloff Distance=false
    Origin Offset=<0 0 -70>
    Override Seed=false
    Velocity Mode=::NewEnumerator1
    Velocity Speed=&Multiply_Float(
     A=&MergeHelper_FloatSwitch(
      0=&RandomRangeFloat(Minimum=241.2247, Maximum=757.0768)
      1=&RandomRangeFloat(Minimum=10, Maximum=9.2722)
      2=?
      3=?
      Index=&SpawnGroup()
      Num=::NewEnumerator0
     )
     B=&FloatFromCurve(
      FloatCurve=DataInterface() {propertyValues {Curve {keys [<at 0.0809 1.4>,<at 0.2058 0.8>,<at 0.336 1.2>,<at 0.4985 0.8>,<at 0.7144 1.2>,<at 0.8629 1>]}}}
      CurveIndex=$Emitter.NormalizedLoopAge
      Scale Curve=1.5
     )
    )
   )
   &InitialMeshOrientation(Use Rotation Vector=false)
  ]
  particleUpdateScript [
   &ParticleState(DeltaTime:NiagaraFloat=&Multiply_Float(A=$Engine.DeltaTime, B=$User.SpeedCurve))
   &UpdateMeshOrientation(
    Rotation Rate=&FloatFromCurve(FloatCurve=DataInterface() {propertyValues {Curve {keys [<at 0 1.5>,<at 1 1>]}}})
    Rotation Vector=&RandomRangeVector(Minimum=<-1 -1 -1>, Maximum=<1 1 1>)
   )
   &ScaleMeshSize(Scale Factor=&Multiply_VectorByFloat(
     Vector=<1 1 1>
     Float=&MergeHelper_FloatSwitch(
      0=&FloatFromCurve(
       FloatCurve=DataInterface() {propertyValues {Curve {keys [
           {interpMode=RCIM_Cubic, tangentMode=RCTM_Break, time=0, value=0.0706}
           {arriveTangent=-0.2191, interpMode=RCIM_Cubic, leaveTangent=-0.2191, tangentMode=RCTM_User, time=0.3256, value=0.8706}
           {arriveTangent=-0.0583, interpMode=RCIM_Cubic, leaveTangent=-0.0583, tangentMode=RCTM_User, time=0.7849, value=0.8618}
           {arriveTangent=-1.7206, interpMode=RCIM_Cubic, leaveTangent=-1.7206, tangentMode=RCTM_User, time=1, value=0}
          ]}}}
       Scale Curve=0.2
      )
      1=&FloatFromCurve(
       FloatCurve=DataInterface() {propertyValues {Curve {keys [
           {tangentMode=RCTM_User, time=0, value=1}
           {arriveTangent=-0.0405, interpMode=RCIM_Cubic, leaveTangent=-0.0405, tangentMode=RCTM_User, time=0.7594, value=0.9647}
           {arriveTangent=-3.3468, interpMode=RCIM_Cubic, leaveTangent=-3.3468, tangentMode=RCTM_User, time=1, value=0}
          ]}}}
       Scale Curve=0.2
      )
      2=?
      3=?
      Index=&SpawnGroup()
      Num=::NewEnumerator0
     )
    ))
   &Drag(Drag=2, Use Linear Drag=true, Use Rotational Drag=false)
   &InheritSourceMovement(ApplyForce=false, ApplyPosition=true, ApplyPositionFalloff=false)
   &GravityForce(Gravity=&Multiply_VectorByFloat(Vector=&RandomRangeVector(Minimum=<0 0 -980>, Maximum=<0 0 -1200>), Float=1.5))
   &Collision(
    Age Colliding Particles=false
    CPU Collision Trace Channel=ECC_GameTraceChannel5
    Collision Enabled=&IsPlatformSetActive(New Platform Set=DataInterface() {propertyValues {Platforms {qualityLevelMask=-4}}})
    Friction=0.5
    KillOnCollision=false
    Module.Control Rest State Via Surface Normal=false
    Module.EnableMaxCollisionCount=false
    Module.Randomize Collision Normal Vector=false
    Restitution=0.5
    Restitution Coefficient Blending:ENiagaraRestitutionMergeType=::NewEnumerator2
   )
   &SolveForcesAndVelocity(
    Clamp Velocity=false
    DeltaTime:NiagaraFloat=&Multiply_Float(A=$Transient.PhysicsDeltaTime, B=$User.SpeedCurve)
    Limit Acceleration=false
   )
   &ScaleColor(
    Scale RGB=&Multiply_VectorByFloat(
     Vector=<1 1 1>
     Float=&MakeCustomFloatFromBool(
      Bool=$User.bIsFrontEndPreview
      True Float=&RandomRangeFloat(Minimum=4.5, Maximum=5)
      False Float=&RandomRangeFloat(Minimum=0.5, Maximum=1.5)
     )
    )
    ScaleA=true
    ScaleRGB=true
   )
   &CameraOffset(Camera Offset Amount=-10)
   &JitterPosition(Jitter Amount:NiagaraFloat=&MergeHelper_FloatSwitch(
     0=0
     1=&FloatFromCurve(
      FloatCurve=DataInterface() {propertyValues {Curve {defaultValue=1, keys [{tangentMode=RCTM_User, time=0, value=0}, {tangentMode=RCTM_User, time=0.4087, value=1}]}}}
      CurveIndex=$Engine.System.Age
      Scale Curve=2
     )
     2=?
     3=?
     Index=&SpawnGroup()
     Num=::NewEnumerator0
    ))
  ]
  rendererClasses [&NiagaraMeshRendererProperties]
  renderers [
   {rendererClass=&NiagaraMeshRendererProperties, rendererIndex=0}
   {rendererClass=&NiagaraMeshRendererProperties, rendererIndex=1}
  ]
 }
 Dust {
  emitterUpdateScript [
   &EmitterState(Life Cycle Mode=::NewEnumerator1, Loop Behavior=::NewEnumerator1, Loop Duration=2, UseLoopDelay=false)
   &SpawnBurst_Instantaneous(Spawn Count=14, Spawn Time=$System.BurstDelay, Use Spawn Probability=false, UseLoopCountLimit=false)
   &SpawnRate(
    SpawnGroup=1
    SpawnRate=&FloatFromCurve(
     FloatCurve=DataInterface() {propertyValues {Curve {
        defaultValue=1
        keys [
         {tangentMode=RCTM_User, time=0.0003, value=0.1518}
         {arriveTangent=8.7005, interpMode=RCIM_Cubic, leaveTangent=8.7005, time=0.1317, value=0.3097}
         <at 0.1597 1.5384 RCIM_Cubic>
        ]
       }}}
     CurveIndex=$Emitter.NormalizedLoopAge
     Scale Curve=52.9991
    )
    Use Spawn Probability=false
   ) {module=SpawnRate001}
   &SpawnBurst_Instantaneous(Spawn Count=16, Spawn Group=2, Spawn Time=$System.BurstDelay, Use Spawn Probability=false, UseLoopCountLimit=false) {module=SpawnBurst_Instantaneous002}
   &SpawnBurst_Instantaneous(Spawn Count=20, Spawn Group=3, Spawn Time=$System.BurstDelay, Use Spawn Probability=false, UseLoopCountLimit=false) {module=SpawnBurst_Instantaneous001}
   &SpawnBurst_Instantaneous(Spawn Group=4, Spawn Time=$System.BurstDelay, Use Spawn Probability=false, UseLoopCountLimit=false) {module=SpawnBurst_Instantaneous003}
  ]
  particleSpawnScript [
   /SetVariables(Particles.VisibilityTag=&SpawnGroup()) {isSetParameters=true}
   &InitializeParticle(
    Color=&MergeHelper_ColorSwitch(
     0=&RandomRangeLinearColor(Minimum=<rgba 0.0938 0.0351 0.0078 1.2>, Maximum=<rgba 0.0931 0.0356 0.008 1>)
     1=&RandomRangeLinearColor(Minimum=<rgba 0.0938 0.0351 0.0078 1.2>, Maximum=<rgba 0.0931 0.0356 0.008 1>)
     2=<rgba 10 1.0204 0 1>
     3=<rgba 1 0.7581 0 1>
     4=<rgba 10 3.9047 0 0.4435>
     Index=&SpawnGroup()
     Num=::NewEnumerator3
    )
    Lifetime=&MergeHelper_FloatSwitch(
     0=&RandomRangeFloat(Minimum=1, Maximum=2)
     1=&RandomRangeFloat(Minimum=1, Maximum=2)
     2=&RandomRangeFloat(Minimum=0.7028, Maximum=0.917)
     3=/Niagara/DynamicInputs/Helpers/DistributionCurveFloat(
      Minimum:NiagaraFloat=0.07
      Maximum:NiagaraFloat=0.14
      Distribution Curve:NiagaraDataInterfaceCurve=DataInterface() {propertyValues {Curve {
         defaultValue=0
         keys [
          {tangentMode=RCTM_User, time=0, value=0}
          {interpMode=RCIM_Cubic, tangentMode=RCTM_User, time=0.905, value=0.0656}
          {tangentMode=RCTM_User, time=1.0025, value=0.9176}
         ]
        }}}
      Curve Parameter Type:NiagaraInt32=0
      "Curve Parameter Type:":NiagaraInt32=0
      Curve Parameter:NiagaraFloat=?
      Evaluation Type:ENiagaraRandomnessEvaluation=::NewEnumerator0
      Recalculate Random Each Loop:NiagaraBool=true
      Randomness Mode:ENiagaraRandomnessMode=::NewEnumerator0
      Random Seed:NiagaraInt32=0
      Fixed Random Seed:NiagaraBool=false
     )
     4=0.04
     Index=&SpawnGroup()
     Num=::NewEnumerator3
    )
    Position Offset=<-20 0 0>
    Sprite Rotation Mode=::NewEnumerator2
    Sprite Size=&MergeHelper_Vector2Switch(
     0=&RandomRangeVector2D(Minimum=<150 150>, Maximum=<250 250>)
     1=&RandomRangeVector2D(Minimum=<150 150>, Maximum=<250 250>)
     2=&RandomRangeVector2D(Minimum=<1 2>, Maximum=<1.6 4>)
     3=&RandomRangeVector2D(Minimum=<61.9732 95.6632>, Maximum=<44.7869 67.4387>)
     4=&Multiply_Vector2D_ByFloat(Float=732.9234, Vector2D=<1 1>)
     Index=&SpawnGroup()
     Num=::NewEnumerator3
    )
    Sprite Size Mode=::NewEnumerator4
    UsePositionOffset=true
    Write Material Random=true
   )
   &SubUVAnimation(End Frame=3, SubUV Animation Mode:ENiagara_SubUVLookupMode=::NewEnumerator2)
   &CameraOffset(Camera Offset Amount=&MaskFloatBySpawnGroup(Masked Float=24.9777, Passthrough Float=0, Spawn Group Mask=4))
   &ShapeLocation(
    MaskSpawnGroup=false
    Non Uniform Scale=<1 1 0.5>
    Offset=<-20 0 5>
    Offset Mode=::NewEnumerator0
    Override Seed=false
    Rotation Coordinate Space=::World
    Shape Origin=&MakeVector(X=0, Y=0, Z=0)
    Sphere Radius=&MergeHelper_FloatSwitch(
     0=60
     1=&FloatFromCurve(
      FloatCurve=DataInterface() {propertyValues {Curve {defaultValue=1, keys [{tangentMode=RCTM_User, time=0, value=0.3013}, {tangentMode=RCTM_User, time=0.8823, value=1}]}}}
      CurveIndex=$System.Age
      Scale Curve=60
     )
     2=60
     3=60
     4=0
     Index=&SpawnGroup()
     Num=::NewEnumerator3
    )
    UseOwnerScale=true
   )
   &AddVelocityFromPoint(
    Module.Use Falloff Distance=false
    Origin Offset=<-20 0 -70>
    Velocity Origin:Vector3f=&MakeVector(X=0, Y=0, Z=0)
    Velocity Strength=&Multiply_Float(
     A=&MergeHelper_FloatSwitch(
      0=&RandomRangeFloat(Minimum=457.6458, Maximum=945.796)
      1=134.5067
      2=&RandomRangeFloat(Minimum=1088.8407, Maximum=2201.8193)
      3=&RandomRangeFloat(Minimum=748.2576, Maximum=1587.2385)
      4=0
      Index=&SpawnGroup()
      Num=::NewEnumerator3
     )
     B=1
    )
   )
  ]
  particleUpdateScript [
   &ParticleState(DeltaTime:NiagaraFloat=&Multiply_Float(A=$Engine.DeltaTime, B=$User.SpeedCurve))
   &SpriteRotationRate(Rotation Rate=&RandomRangeFloat(Minimum=-50, Maximum=50))
   &GravityForce(Gravity=&MergeHelper_VectorSwitch(
     0=&RandomRangeVector(Minimum=<0 0 -980>, Maximum=<0 0 -1200>)
     1=&RandomRangeVector(Minimum=<0 0 -980>, Maximum=<0 0 -1200>)
     2=<0 0 0>
     3=<0 0 -1375.6633>
     4=<0 0 0>
     Index=&SpawnGroup()
     Num=::NewEnumerator3
    ))
   &Drag(
    Drag=&MergeHelper_FloatSwitch(
     0=1
     1=1
     2=&FloatFromCurve(
      FloatCurve=DataInterface() {propertyValues {Curve {
         defaultValue=0
         keys [
          {arriveTangent=4.1785, interpMode=RCIM_Cubic, leaveTangent=4.1785, tangentMode=RCTM_User, time=0, value=0}
          {arriveTangent=1.1179, interpMode=RCIM_Cubic, leaveTangent=1.1179, tangentMode=RCTM_User, time=0.2042, value=0.7555}
          {arriveTangent=0.0224, interpMode=RCIM_Cubic, leaveTangent=0.0224, tangentMode=RCTM_User, time=1, value=1}
         ]
        }}}
      Scale Curve=&RandomRangeFloat(Minimum=9.5392, Maximum=12.466)
     )
     3=&FloatFromCurve(
      FloatCurve=DataInterface() {propertyValues {Curve {
         defaultValue=0
         keys [
          {arriveTangent=4.1785, interpMode=RCIM_Cubic, leaveTangent=4.1785, tangentMode=RCTM_User, time=0, value=0}
          {arriveTangent=1.1179, interpMode=RCIM_Cubic, leaveTangent=1.1179, tangentMode=RCTM_User, time=0.2042, value=0.7555}
          {arriveTangent=0.0224, interpMode=RCIM_Cubic, leaveTangent=0.0224, tangentMode=RCTM_User, time=1, value=1}
         ]
        }}}
      Scale Curve=&RandomRangeFloat(Minimum=1.0307, Maximum=2.6871)
     )
     4=0
     Index=&SpawnGroup()
     Num=::NewEnumerator3
    )
    Rotational Drag=5
    Use Linear Drag=true
    Use Rotational Drag=true
   ) {module=Drag001}
   &CurlNoiseForce(
    Bias Noise=false
    Noise Frequency=39.278
    Noise Strength=&MergeHelper_FloatSwitch(
     0=0
     1=0
     2=&FloatFromCurve(
      FloatCurve=DataInterface() {propertyValues {Curve {
         defaultValue=1
         keys [
          {arriveTangent=4.1785, interpMode=RCIM_Cubic, leaveTangent=4.1785, tangentMode=RCTM_User, time=0, value=0}
          {arriveTangent=1.1179, interpMode=RCIM_Cubic, leaveTangent=1.1179, tangentMode=RCTM_User, time=0.1435, value=0.7781}
          {arriveTangent=0.0224, interpMode=RCIM_Cubic, leaveTangent=0.0224, tangentMode=RCTM_User, time=1, value=1}
         ]
        }}}
      Scale Curve=600
     )
     3=?
     Index=&SpawnGroup()
     Num=::NewEnumerator1
    )
    Pan Noise=false
   )
   &SolveForcesAndVelocity(
    Clamp Velocity=false
    DeltaTime:NiagaraFloat=&Multiply_Float(A=$Engine.DeltaTime, B=$User.SpeedCurve)
    Limit Acceleration=false
   )
   &ScaleSpriteSizeBySpeed(Max Scale Factor=&MergeHelper_Vector2Switch(0=<1 1>, 1=<1 1>, 2=<1 10.901>, 3=<1 3.4565>, 4=?, Index=&SpawnGroup(), Num=::NewEnumerator2))
   &ScaleSpriteSize(Uniform Curve Sprite Scale=DataInterface() {propertyValues {Curve {keys [
        {arriveTangent=4.1785, interpMode=RCIM_Cubic, leaveTangent=4.1785, tangentMode=RCTM_User, time=0, value=0}
        {arriveTangent=0.9995, interpMode=RCIM_Cubic, leaveTangent=0.9995, tangentMode=RCTM_User, time=0.2223, value=0.6825}
        {arriveTangent=-0.0405, interpMode=RCIM_Cubic, leaveTangent=-0.0405, tangentMode=RCTM_User, time=1, value=1}
       ]}}})
   &ScaleColor(
    Scale Alpha=&FloatFromCurve(
     FloatCurve=DataInterface() {propertyValues {Curve {keys [
         {arriveTangent=-0.4559, interpMode=RCIM_Cubic, leaveTangent=-0.4559, tangentMode=RCTM_User, time=0, value=1}
         {arriveTangent=-2.4378, interpMode=RCIM_Cubic, leaveTangent=-2.4378, tangentMode=RCTM_User, time=1.0011, value=0.002}
        ]}}}
     Scale Curve=0.8
    )
    Scale RGB=&Multiply_VectorByFloat(Vector=<1 1 1>, Float=&RandomRangeFloat(Minimum=0.5, Maximum=1))
    ScaleA=true
    ScaleRGB=true
   )
   /SetVariables(Particles.Position2:NiagaraPosition=&AddVectorToPosition(Position=$Engine.Emitter.SimulationPosition, Vector=Linked() {linkedVariable {name=Particles.Position, type=NiagaraPosition}})) {isSetParameters=true}
   /SetVariables() {isSetParameters=true}
  ]
  rendererClasses [&NiagaraSpriteRendererProperties]
  renderers [
   {rendererClass=&NiagaraSpriteRendererProperties}
   {rendererClass=&NiagaraSpriteRendererProperties, rendererIndex=1}
   {rendererClass=&NiagaraSpriteRendererProperties, rendererIndex=2}
   {rendererClass=&NiagaraSpriteRendererProperties, rendererIndex=3}
   {rendererClass=&NiagaraSpriteRendererProperties, rendererIndex=4}
  ]
 }
 Flare_Complex {
  emitterUpdateScript [
   &EmitterState(Life Cycle Mode=::NewEnumerator1, Loop Behavior=::NewEnumerator1, Loop Delay=0.075, Loop Duration=2, UseLoopDelay=true)
   &SpawnBurst_Instantaneous(Use Spawn Probability=false, UseLoopCountLimit=false) {module=SpawnBurst_Instantaneous003}
   &SpawnBurst_Instantaneous(Spawn Group=1, Use Spawn Probability=false, UseLoopCountLimit=false) {module=SpawnBurst_Instantaneous001}
   &SpawnBurst_Instantaneous(Spawn Group=2, Spawn Time=0.85, Use Spawn Probability=false, UseLoopCountLimit=false)
   &SpawnBurst_Instantaneous(Spawn Group=3, Spawn Time=0.85, Use Spawn Probability=false, UseLoopCountLimit=false) {module=SpawnBurst_Instantaneous002}
  ]
  particleSpawnScript [
   /SetVariables(Particles.VisibilityTag=&SpawnGroup(Max Spawn Group=2)) {isSetParameters=true}
   &InitializeParticle(
    Color=&MakeCustomLinearColorFromBool(
     Bool=&SetBoolByIntComparison(A=&SpawnGroup(), B=1, Comparison Type=::NewEnumerator7)
     True Color=&SelectLinearColorFromArray(
      Array Sampling Mode=::NewEnumerator1
      Direct Array Index=&SpawnGroup()
      Color Selection Array=DataInterface() {propertyValues {ColorData [<rgba 1 1 1 1>, <rgba 1 1 1 0.5>, <rgba 1 1 1 1>, <rgba 100 61.7074 35.1155 1>]}}
      Random Seed=?
     )
     False Color=<rgba 1 1 1 1>
    )
    Lifetime=&MaskFloatBySpawnGroup(
     Masked Float=0.236
     Passthrough Float=&MaskFloatBySpawnGroup(Masked Float=-0.326, Passthrough Float=0.178, Spawn Group Mask=1)
    )
    Mesh Scale Mode=::NewEnumerator3
    Mesh Uniform Scale=&MaskFloatBySpawnGroup(Masked Float=2, Passthrough Float=1, Spawn Group Mask=2)
    Sprite Rotation Angle=&MaskFloatBySpawnGroup(Masked Float=0, Passthrough Float=&RandomRangeFloat(Minimum=0, Maximum=1))
    Sprite Rotation Mode=::NewEnumerator1
    Sprite Size=&Vector2DFromFloat(Value=500)
    Sprite Size Mode=::NewEnumerator4
    UsePositionOffset=true
    Write Material Random=true
   )
   &CameraOffset(Camera Offset Amount=&MaskFloatBySpawnGroup(Masked Float=50, Passthrough Float=80))
  ]
  particleUpdateScript [
   &ParticleState(DeltaTime:NiagaraFloat=&Multiply_Float(A=$Engine.DeltaTime, B=$User.SpeedCurve))
   &SpriteRotationRate(
    Delta Time:NiagaraFloat=&FloatFromCurve(FloatCurve=DataInterface() {propertyValues {Curve {keys [
         {tangentMode=RCTM_User, time=0, value=1}
         {arriveTangent=-0.3174, interpMode=RCIM_Cubic, leaveTangent=-0.3174, tangentMode=RCTM_User, time=0.3, value=1}
         {arriveTangent=0.0123, interpMode=RCIM_Cubic, leaveTangent=0.0123, tangentMode=RCTM_User, time=1, value=0}
        ]}}})
    Rotation Rate=&MaskFloatBySpawnGroup(Masked Float=5, Passthrough Float=0, Spawn Group Mask=1)
   )
   &ScaleSpriteSize(
    Scale Factor=&MaskVector2dBySpawnGroup(
     Masked Vector2D=&Vector2DFromCurve(
      Vector2Curve=DataInterface() {propertyValues {
        XCurve {keys [<at 0 1 RCIM_Cubic>, {arriveTangent=-3.1442, interpMode=RCIM_Cubic, leaveTangent=-3.1442, tangentMode=RCTM_User, time=1, value=0}]}
        YCurve {keys [
          {arriveTangent=4.8145, interpMode=RCIM_Cubic, time=0, value=1}
          {arriveTangent=-3.0695, interpMode=RCIM_Cubic, leaveTangent=-3.0695, tangentMode=RCTM_User, time=1, value=0}
         ]}
       }}
      ScaleCurve=<1 1>
     )
     Passthrough Vector2D=&Vector2DFromFloat(Value=&MaskFloatBySpawnGroup(
       Masked Float=&FloatFromCurve(FloatCurve=DataInterface() {propertyValues {Curve {keys [
            {arriveTangent=5.1291, interpMode=RCIM_Cubic, leaveTangent=5.1291, tangentMode=RCTM_User, time=0, value=0}
            {arriveTangentWeight=0.1274, interpMode=RCIM_Cubic, leaveTangentWeight=0.7778, tangentMode=RCTM_User, time=0.5, value=1}
           ]}}})
       Passthrough Float=&FloatFromCurve(FloatCurve=DataInterface() {propertyValues {Curve {keys [
            {arriveTangent=3.3451, interpMode=RCIM_Cubic, leaveTangent=3.3451, tangentMode=RCTM_User, time=0, value=0.5}
            {arriveTangentWeight=0.1274, interpMode=RCIM_Cubic, leaveTangentWeight=0.7778, tangentMode=RCTM_User, time=0.5, value=1}
           ]}}})
       Spawn Group Mask=1
      ))
     Spawn Group Mask=0
    )
    Scale Sprite Size Mode=::NewEnumerator2
   )
   &Color(Color=&MakeCustomLinearColorFromBool(
     Bool=&SetBoolByIntComparison(A=&SpawnGroup(), B=1, Comparison Type=::NewEnumerator7)
     True Color=&SelectLinearColorFromArray(
      Array Sampling Mode=::NewEnumerator1
      Direct Array Index=&SpawnGroup()
      Color Selection Array=DataInterface() {propertyValues {ColorData [<rgba 1 1 1 1>, <rgba 1 1 1 0.5>, <rgba 1 1 1 1>, <rgba 100 61.7074 35.1155 1>]}}
      Random Seed=?
     )
     False Color=<rgba 1 1 1 1>
    ))
   &ScaleMeshSize(Scale Factor=&VectorFromFloat(Value=&MaskFloatBySpawnGroup(
      Masked Float=&FloatFromCurve(FloatCurve=DataInterface() {propertyValues {Curve {
          defaultValue=0
          keys [
           {arriveTangent=4.1785, interpMode=RCIM_Cubic, leaveTangent=4.1785, tangentMode=RCTM_User, time=0, value=0}
           {arriveTangent=1.529, interpMode=RCIM_Cubic, leaveTangent=1.529, tangentMode=RCTM_User, time=0.2042, value=0.6228}
           {arriveTangent=0.0224, interpMode=RCIM_Cubic, leaveTangent=0.0224, tangentMode=RCTM_User, time=1, value=1}
          ]
         }}})
      Passthrough Float=&FloatFromCurve(FloatCurve=DataInterface() {propertyValues {Curve {
          defaultValue=0
          keys [
           <at 0 0 RCIM_Cubic>
           {arriveTangent=1.6664, interpMode=RCIM_Cubic, leaveTangent=1.6664, tangentMode=RCTM_User, time=0.45, value=0.2}
           {arriveTangent=1.1179, interpMode=RCIM_Cubic, leaveTangent=1.1179, tangentMode=RCTM_User, time=0.55, value=0.75}
           {arriveTangent=0.0224, interpMode=RCIM_Cubic, leaveTangent=0.0224, tangentMode=RCTM_User, time=1, value=1}
          ]
         }}})
      Spawn Group Mask=2
     )))
   &ScaleColor(
    Scale Alpha=&MaskFloatBySpawnGroup(
     Masked Float=&FloatFromCurve(FloatCurve=DataInterface() {propertyValues {Curve {keys [
          {arriveTangent=-23.168, interpMode=RCIM_Cubic, leaveTangent=-23.168, tangentMode=RCTM_User, time=0, value=4}
          {arriveTangent=-4.5789, interpMode=RCIM_Cubic, leaveTangent=-4.5789, tangentMode=RCTM_User, time=0.2, value=1}
          <at 1 0 RCIM_Cubic>
         ]}}})
     Passthrough Float=&MaskFloatBySpawnGroup(
      Masked Float=&FloatFromCurve(
       FloatCurve=DataInterface() {propertyValues {Curve {
          defaultValue=0
          keys [
           {arriveTangent=9.0725, interpMode=RCIM_Cubic, leaveTangent=9.0725, tangentMode=RCTM_User, time=0, value=0}
           {arriveTangent=-2.5249, interpMode=RCIM_Cubic, leaveTangent=-2.5249, tangentMode=RCTM_User, time=0.1408, value=0.9847}
           {arriveTangent=-0.0405, interpMode=RCIM_Cubic, leaveTangent=-0.0405, tangentMode=RCTM_User, time=1, value=0}
          ]
         }}}
       Scale Curve=0.2
      )
      Passthrough Float=&MaskFloatBySpawnGroup(
       Masked Float=&FloatFromCurve(
        FloatCurve=DataInterface() {propertyValues {Curve {keys [<at 0 1 RCIM_Cubic>,<at 1 0 RCIM_Cubic>]}}}
        Scale Curve=&MaskFloatBySpawnGroup(Masked Float=0.2, Passthrough Float=1, Spawn Group Mask=2)
       )
       Passthrough Float=&FloatFromCurve(
        FloatCurve=DataInterface() {propertyValues {Curve {keys [<at 0 1 RCIM_Cubic>,<at 1 0 RCIM_Cubic>]}}}
        Scale Curve=&MaskFloatBySpawnGroup(Masked Float=0.2, Passthrough Float=1, Spawn Group Mask=2)
       )
       Spawn Group Mask=2
      )
      Spawn Group Mask=1
     )
    )
    Scale RGB=&MakeVectorFromLinearColorRGB(Color=&MakeCustomLinearColorFromBool(
      Bool=&SetBoolByIntComparison(A=&SpawnGroup(), B=1, Comparison Type=::NewEnumerator7)
      True Color=&ColorFromCurve(ColorCurve=DataInterface() {propertyValues {
         AlphaCurve {keys []}
         BlueCurve {keys [{arriveTangent=-1.5872, interpMode=RCIM_Cubic, leaveTangent=-1.5872, tangentMode=RCTM_User, time=0, value=3}, <at 1 0.5 RCIM_Cubic>]}
         GreenCurve {keys [{arriveTangent=-4.9882, interpMode=RCIM_Cubic, leaveTangent=-4.9882, tangentMode=RCTM_User, time=0, value=3.75}, <at 1 0.625 RCIM_Cubic>]}
         RedCurve {keys [{arriveTangent=-3.5785, interpMode=RCIM_Cubic, leaveTangent=-3.5785, tangentMode=RCTM_User, time=0, value=4}, <at 1 1 RCIM_Cubic>]}
        }})
      False Color=<rgba 1 1 1 1>
     ))
    ScaleA=true
    ScaleRGB=true
   )
  ]
  rendererClasses [&NiagaraMeshRendererProperties]
  renderers [{rendererClass=&NiagaraMeshRendererProperties, rendererIndex=0}]
 }
 Flash {
  emitterUpdateScript [
   &EmitterState(Life Cycle Mode=::NewEnumerator1, Loop Behavior=::NewEnumerator1, Loop Duration=1.55, UseLoopDelay=false)
   &SpawnBurst_Instantaneous(Spawn Time=0.8, Use Spawn Probability=false, UseLoopCountLimit=false)
   &SpawnBurst_Instantaneous(Spawn Time=0.91, Use Spawn Probability=false, UseLoopCountLimit=false) {module=SpawnBurst_Instantaneous001}
  ]
  particleSpawnScript [
   &InitializeParticle(
    Color=<rgba 5 1.9632 0 0.0783>
    Lifetime=0.2
    Mesh Scale Mode=::NewEnumerator3
    Mesh Uniform Scale=4.4726
    Sprite Size Mode=::NewEnumerator3
    Uniform Sprite Size=400
    UsePositionOffset=false
    Write Material Random=true
   )
   &CameraOffset(Camera Offset Amount=40)
  ]
  particleUpdateScript [
   &ParticleState(DeltaTime:NiagaraFloat=&Multiply_Float(A=$Engine.DeltaTime, B=$User.SpeedCurve))
   &ScaleMeshSize(Scale Factor=&VectorFromFloat(Value=&FloatFromCurve(FloatCurve=DataInterface() {propertyValues {Curve {keys [
          {arriveTangent=0.7693, interpMode=RCIM_Cubic, leaveTangent=0.7693, tangentMode=RCTM_User, time=0.0072, value=0.6588}
          {arriveTangent=0.1864, interpMode=RCIM_Cubic, leaveTangent=0.1864, tangentMode=RCTM_User, time=1, value=1}
         ]}}})))
   &ScaleColor(
    Scale Alpha=&FloatFromCurve(
     FloatCurve=DataInterface() {propertyValues {Curve {
        defaultValue=1
        keys [
         {arriveTangent=4.1785, interpMode=RCIM_Cubic, leaveTangent=4.1785, tangentMode=RCTM_User, time=0, value=0}
         {arriveTangent=1.1179, interpMode=RCIM_Cubic, leaveTangent=1.1179, tangentMode=RCTM_User, time=0.1467, value=0.7976}
         {arriveTangent=-2.0201, interpMode=RCIM_Cubic, leaveTangent=-2.0201, time=0.3271, value=1.0388}
         {arriveTangent=0.0224, interpMode=RCIM_Cubic, leaveTangent=0.0224, tangentMode=RCTM_User, time=0.5319, value=0.0194}
        ]
       }}}
     Scale Curve=2.5
    )
    ScaleA=true
    ScaleRGB=false
   )
   &DynamicMaterialParameters(
    Index 0 Param 1=&FloatFromCurve(FloatCurve=DataInterface() {propertyValues {Curve {keys [
         {interpMode=RCIM_Cubic, tangentMode=RCTM_User, time=0.5, value=0}
         {arriveTangent=3.8908, interpMode=RCIM_Cubic, leaveTangent=3.8908, tangentMode=RCTM_User, time=1, value=1}
        ]}}})
    Index 0 Param 2=&FloatFromCurve(FloatCurve=DataInterface() {propertyValues {Curve {keys [
         {arriveTangent=-0.6134, interpMode=RCIM_Cubic, leaveTangent=-0.6134, tangentMode=RCTM_User, time=0.0062, value=2.8434}
         {arriveTangent=-3.248, interpMode=RCIM_Cubic, leaveTangent=-3.248, tangentMode=RCTM_User, time=1, value=0.5}
        ]}}})
    Index 0 Param 3=&FloatFromCurve(FloatCurve=DataInterface() {propertyValues {Curve {keys [
         {arriveTangent=-1.2425, interpMode=RCIM_Cubic, time=0.0071, value=1}
         {arriveTangent=-2.1091, interpMode=RCIM_Cubic, leaveTangent=-2.1091, tangentMode=RCTM_User, time=1, value=0}
        ]}}})
    Index 0 Param 4=&FloatFromCurve(
     FloatCurve=DataInterface() {propertyValues {Curve {keys [
         {arriveTangent=2.3563, interpMode=RCIM_Cubic, leaveTangent=2.3563, tangentMode=RCTM_User, time=0, value=-0.0941}
         {tangentMode=RCTM_User, time=1, value=1}
        ]}}}
     Scale Curve=2
    )
    Module.Index0Param0Write=true
    Module.Index0Param1Write=true
    Module.Index0Param2Write=true
    Module.Index0Param3Write=true
   ) {module=DynamicMaterialParameters001}
  ]
  rendererClasses [&NiagaraMeshRendererProperties]
  renderers [{rendererClass=&NiagaraMeshRendererProperties, rendererIndex=0}]
 }
 Glow {
  emitterUpdateScript [
   &EmitterState(Life Cycle Mode=::NewEnumerator1, Loop Behavior=::NewEnumerator1, Loop Duration=2, UseLoopDelay=false)
   &SpawnBurst_Instantaneous(Spawn Time=$System.BurstDelay, Use Spawn Probability=false, UseLoopCountLimit=false)
  ]
  particleSpawnScript [
   &InitializeParticle(
    Color=<rgba 0.4792 0.1547 0 0.8348>
    Lifetime=3
    Sprite Size Mode=::NewEnumerator3
    Uniform Sprite Size=303.7185
    UsePositionOffset=false
    Write Material Random=true
   )
   &DynamicMaterialParameters(
    Index 0 Param 1=3
    Index 0 Param 2=&FN_FrontEndFloatModifier(Gameplay=100, Front End=30)
    Module.Index0Float0Write=true
    Module.Index0Float1Write=true
    Module.Index0Float2Write=true
    Module.Index0Float3Write=true
   )
   &CameraOffset(Camera Offset Amount=25.7503)
  ]
  particleUpdateScript [
   &ParticleState()
   &ScaleSpriteSize(
    Scale Sprite Size Mode=::NewEnumerator0
    Uniform Scale Factor=&FN_FrontEndFloatModifier(Gameplay=1, Front End=1)
   )
   &ScaleColor(
    Scale Alpha=&FloatFromCurve(
     FloatCurve=DataInterface() {propertyValues {Curve {
        defaultValue=1
        keys [
         {arriveTangent=-0.9545, interpMode=RCIM_Cubic, leaveTangent=-0.9545, tangentMode=RCTM_User, time=-0.2488, value=-0.1215}
         {arriveTangent=-0.0599, interpMode=RCIM_Cubic, leaveTangent=-0.0599, tangentMode=RCTM_User, time=-0.2037, value=1.215}
         {arriveTangent=-1.2597, interpMode=RCIM_Cubic, leaveTangent=-1.2597, time=-0.1372, value=0.6182}
         {arriveTangent=-0.9584, interpMode=RCIM_Cubic, leaveTangent=-0.9584, tangentMode=RCTM_User, time=0.3962, value=0.4593}
         {interpMode=RCIM_Cubic, tangentMode=RCTM_User, time=0.7512, value=-0.1215}
        ]
       }}}
     Scale Curve=&Multiply_Float(
      A=&FN_FrontEndFloatModifier(Gameplay=3, Front End=1)
      B=&FN_PickFloatDependOnPlatform(
       TargetPlatform=DataInterface() {propertyValues {Platforms {
          qualityLevelMask=-4
          deviceProfileStates [{profileName=Windows, qualityLevelMask=2, setQualityLevelMask=2}, {profileName=Mac, qualityLevelMask=2, setQualityLevelMask=2}]
         }}}
       TargetPlatformActive=1
       TargetPlatformNoActive=0
      )
     )
    )
    ScaleA=true
    ScaleRGB=true
   )
   &FN_Light_Attributes(
    Color=<rgba 60 25.6172 0 1>
    Create Color Attribute=true
    Create Position Attribute=true
    Create Radius Attribute=true
    Position=<0 0 20>
    Radius=150
    Scale Color=&FloatFromCurve(FloatCurve=DataInterface() {propertyValues {Curve {keys [<at 0 1>,<at 0.1558 1.0029>,<at 0.2906 0.3559>,<at 1 0>]}}})
   )
  ]
  rendererClasses [&NiagaraSpriteRendererProperties, &NiagaraLightRendererProperties]
  renderers [
   {rendererClass=&NiagaraSpriteRendererProperties}
   {rendererClass=&NiagaraLightRendererProperties, rendererIndex=1}
  ]
 }
 GroundCrack {
  emitterSpawnScript [
   /SetVariables(Emitter.HoldPositionEmitter:NiagaraPosition=/Niagara/DynamicInputs/Helpers/SimulationPosition()) {isSetParameters=true, module=SetVariables_#3}
   /SetVariables(Emitter.FRONT_END_USE_WPO:NiagaraFloat=&FN_FrontEndFloatModifier(Gameplay=1, Front End=0)) {isSetParameters=true}
   /SetVariables(Emitter.FrontEndColorMod:NiagaraFloat=&FN_FrontEndFloatModifier(Gameplay=0.6, Front End=0.5)) {isSetParameters=true}
   /SetVariables(Emitter.FRONTENDHEIGHTMOD:NiagaraFloat=&FN_FrontEndFloatModifier(Gameplay=-0.04, Front End=-0.01)) {isSetParameters=true}
  ]
  emitterUpdateScript [
   &EmitterState(
    Inactive Response=::NewEnumerator1
    Life Cycle Mode=::NewEnumerator1
    Loop Behavior=::NewEnumerator1
    Loop Duration=4.5595
    UseLoopDelay=false
   )
   &SpawnBurst_Instantaneous(Use Spawn Probability=false, UseLoopCountLimit=false)
  ]
  particleSpawnScript [
   &InitializeParticle(
    Color=<rgba 0.0723 0.0168 0.0003 1>
    Lifetime=3.2
    Mass Max=2
    Mass Mode=::NewEnumerator3
    Mesh Scale Mode=::NewEnumerator3
    Mesh Uniform Scale=5.7085
    Position Offset=&FN_FrontEndVector3Modifier(Gameplay=<0 0 20>, Front End=<0 0 1>)
    Sprite Rotation Angle=90
    Sprite Rotation Mode=::NewEnumerator0
    Sprite Size Mode=::NewEnumerator3
    Uniform Sprite Size=350
    UsePositionOffset=true
    Write Material Random=true
   )
   &InitialMeshOrientation(Mesh Orientation Mode=::NewEnumerator2, Use Rotation Vector=false)
  ]
  particleUpdateScript [
   &ParticleState()
   &ScaleMeshSize(Scale Factor=&VectorFromFloat(Value=&FloatFromCurve(FloatCurve=DataInterface() {propertyValues {Curve {keys [<at 0 0>,<at 0.02 1.04>,<at 0.07 1>,<at 0.8294 1>]}}}))) {module=ScaleMeshSize001}
   &ScaleMeshSize()
   &DynamicMaterialParameters(
    Index 0 Param 1=&FloatFromCurve(FloatCurve=DataInterface() {propertyValues {Curve {keys [<at 0.0222 1>,<at 0.2839 -0.0471>,<at 0.9 0>,<at 1 1>]}}})
    Index 0 Param 2=&FloatFromCurve(FloatCurve=DataInterface() {propertyValues {Curve {keys [<at 0 1>,<at 0.4793 1>,<at 0.5887 1>,<at 1 1>]}}})
    Module.Index0Float0Write=true
    Module.Index0Float1Write=true
    Module.Index0Float2Write=true
    Module.Index0Float3Write=true
   )
   /SetVariables(
    Particles.DarkGlow:LinearColor=&MultiplyLinearColor(
     Linear Color=<rgba 0.0104 0.0104 0.0104 1.3>
     Scale Factor=&MakeVector4(
      X=1
      Y=1
      Z=1
      W=&FloatFromCurve(FloatCurve=DataInterface() {propertyValues {Curve {keys [<at 0 0>,<at 0.06 0>,<at 0.08 1>,<at 0.8 1>,<at 1 0>]}}})
     )
    )
    Particles.GlowColor=&MultiplyLinearColor(
     Linear Color=<rgba 0.5 0.1206 0 1>
     Scale Factor=&MakeVector4(
      X=1
      Y=1
      Z=1
      W=&FloatFromCurve(
       FloatCurve=DataInterface() {propertyValues {Curve {keys [<at 0 0>,<at 0.06 0>,<at 0.08 1>,<at 0.85 1>,<at 0.98 0>]}}}
       Scale Curve=&FNTimeOfDayScalar(Night Scale=0.8, Morning Scale=1, Evening Scale=1)
      )
     )
    )
   ) {isSetParameters=true, module=SetVariables_#4}
   &ScaleColor(
    Scale Alpha=&FloatFromCurve(FloatCurve=DataInterface() {propertyValues {Curve {keys [<at 0 0>,<at 0.06 0>,<at 0.08 1>,<at 0.9099 1>,<at 1 0>]}}})
    ScaleA=true
    ScaleRGB=true
   )
  ]
  rendererClasses [&NiagaraMeshRendererProperties]
  renderers [
   {rendererClass=&NiagaraMeshRendererProperties, rendererIndex=0}
   {rendererClass=&NiagaraMeshRendererProperties, rendererIndex=1}
   {rendererClass=&NiagaraMeshRendererProperties, rendererIndex=2}
  ]
 }
 Low_Med {
  emitterSpawnScript [/SetVariables(Emitter.HoldPositionEmitter:NiagaraPosition=&SimulationPosition()) {isSetParameters=true, module=SetVariables_#3}]
  emitterUpdateScript [
   &EmitterState(Inactive Response=::NewEnumerator1, Life Cycle Mode=::NewEnumerator1, Loop Behavior=::NewEnumerator1, Loop Duration=3.2, UseLoopDelay=false)
   &SpawnBurst_Instantaneous(Use Spawn Probability=false, UseLoopCountLimit=false)
   /SetVariables(Emitter.OpacityMult:NiagaraFloat=&FloatFromCurve(
     FloatCurve=DataInterface() {propertyValues {Curve {keys [<at 0.0866 1.2>,<at 0.3732 0>,<at 0.9 0>,<at 1 1.2>]}}}
     CurveIndex=$Emitter.NormalizedLoopAge
    )) {isSetParameters=true}
  ]
  particleSpawnScript [
   &InitializeParticle(
    Color=<rgba 0.0103 0.0024 0 1>
    Lifetime=3.23
    Mass Max=2
    Mass Mode=::NewEnumerator3
    Mesh Scale=<6 6 6>
    Mesh Scale Mode=::NewEnumerator4
    Position Offset=<0 0 3>
    Sprite Rotation Angle=119
    Sprite Rotation Mode=::NewEnumerator0
    Sprite Size Mode=::NewEnumerator3
    Uniform Sprite Size=360
    UsePositionOffset=true
    Write Material Random=true
   )
   /Niagara/Modules/Decal/Decal_Attributes(
    Decal Facing Mode=1
    Decal Rotation Mode=3
    Decal Size:Vector3f=<500 360 360>
    Orientation=1
    Rotation Angle=0.33
    Set Alignment=true
    Size=1
   )
   &SpriteFacingAndAlignment(Sprite Facing=<0 0 1>, WriteAlignment=true, WriteFacing=true)
  ]
  particleUpdateScript [
   &ParticleState()
   &DynamicMaterialParameters(
    Index 0 Param 1=&FloatFromCurve(FloatCurve=DataInterface() {propertyValues {Curve {keys [<at 0 1>,<at 0.05 0>,<at 0.9 0>,<at 1 1>]}}})
    Module.Index0Float0Write=true
    Module.Index0Float1Write=true
    Module.Index0Float2Write=true
    Module.Index0Float3Write=true
   )
   /SetVariables(
    Particles.DarkGlow:LinearColor=&MultiplyLinearColor(
     Linear Color=<rgba 0.0104 0.0104 0.0104 10>
     Scale Factor=&MakeVector4(
      X=1
      Y=1
      Z=1
      W=&FloatFromCurve(FloatCurve=DataInterface() {propertyValues {Curve {keys [<at 0 1>,<at 0.8 1>,<at 1 0>]}}}, Scale Curve=5)
     )
    )
    Particles.GlowColor=&MultiplyLinearColor(
     Linear Color=<rgba 0.5 0.1206 0 1>
     Scale Factor=&MakeVector4(
      X=1
      Y=1
      Z=1
      W=&FloatFromCurve(
       FloatCurve=DataInterface() {propertyValues {Curve {keys [<at 0 0>,<at 0.2 1>,<at 0.85 1>,<at 0.98 0>]}}}
       Scale Curve=&FNTimeOfDayScalar(Night Scale=0.8, Morning Scale=1, Evening Scale=1)
      )
     )
    )
   ) {isSetParameters=true, module=SetVariables_#4}
   &ScaleColor(
    Scale Alpha=&FloatFromCurve(FloatCurve=DataInterface() {propertyValues {Curve {keys [<at 0 0>,<at 0.1752 0>,<at 0.3066 1>,<at 0.9099 1>,<at 1 0>]}}})
    ScaleA=true
    ScaleRGB=true
   )
  ]
  rendererClasses [&NiagaraDecalRendererProperties]
  renderers [
   {rendererClass=&NiagaraDecalRendererProperties, rendererIndex=0}
   {rendererClass=&NiagaraDecalRendererProperties, rendererIndex=1}
   {rendererClass=&NiagaraDecalRendererProperties, rendererIndex=2}
  ]
 }
 Smokelow {
  emitterUpdateScript [
   &EmitterState()
   &SpawnBurst_Instantaneous(Spawn Count=15, Spawn Time=$System.BurstDelay, Use Spawn Probability=false, UseLoopCountLimit=false)
   &SpawnBurst_Instantaneous(Spawn Count=18, Spawn Group=1, Spawn Time=$System.BurstDelay, Use Spawn Probability=false, UseLoopCountLimit=false) {module=SpawnBurst_Instantaneous001}
   &SpawnBurst_Instantaneous(Spawn Count=59, Spawn Group=2, Spawn Time=$System.BurstDelay, Use Spawn Probability=false, UseLoopCountLimit=false) {module=SpawnBurst_Instantaneous002}
  ]
  particleSpawnScript [
   /SetVariables(Particles.VisibilityTag=&SpawnGroup()) {isSetParameters=true}
   /SetVariables(Particles.SpriteSize=&Multiply_Vector2D_ByFloat(Float=1.5, Vector2D=<1 1>)) {isSetParameters=true}
   &InitializeParticle(
    Color=&MergeHelper_ColorSwitch(
     0=&RandomRangeLinearColor(Minimum=<rgba 0.026 0.0161 0.0044 1.2>, Maximum=<rgba 0.026 0.0161 0.0044 1.2>)
     1=&RandomRangeLinearColor(Minimum=<rgba 0.026 0.0161 0.0044 1.2>, Maximum=<rgba 0.026 0.0161 0.0044 1.2>)
     2=<rgba 2.2 0.6614 0 1.2>
     3=?
     Index=&SpawnGroup()
     Num=::NewEnumerator1
    )
    Lifetime=&MergeHelper_FloatSwitch(
     0=&RandomRangeFloat(Minimum=0.188, Maximum=0.448)
     1=&DistributionCurveFloat(
      Minimum:NiagaraFloat=0.168
      Maximum:NiagaraFloat=0.25
      Distribution Curve:NiagaraDataInterfaceCurve=DataInterface() {propertyValues {Curve {
         defaultValue=0
         keys [
          {tangentMode=RCTM_User, time=0, value=0}
          {interpMode=RCIM_Cubic, tangentMode=RCTM_User, time=0.1946, value=0.098}
          {tangentMode=RCTM_User, time=1.0025, value=0.9176}
         ]
        }}}
      Curve Parameter Type:NiagaraInt32=0
      "Curve Parameter Type:":NiagaraInt32=0
      Curve Parameter:NiagaraFloat=?
      Evaluation Type:ENiagaraRandomnessEvaluation=::NewEnumerator0
      Recalculate Random Each Loop:NiagaraBool=true
      Randomness Mode:ENiagaraRandomnessMode=::NewEnumerator0
      Random Seed:NiagaraInt32=0
      Fixed Random Seed:NiagaraBool=false
     )
     2=&RandomRangeFloat(Minimum=0.188, Maximum=0.496)
     3=?
     Index=&SpawnGroup()
     Num=::NewEnumerator1
    )
    Sprite Rotation Mode=::NewEnumerator2
    Sprite Size=&MergeHelper_Vector2Switch(
     0=&Multiply_Vector2D_ByFloat(Float=&RandomRangeFloat(Minimum=105.9358, Maximum=133.5037), Vector2D=<1 1>)
     1=/Niagara/DynamicInputs/Helpers/DistributionCurveVector2D(
      Minimum:Vector2f=<10 108.5735>
      Maximum:Vector2f=<12 229.9271>
      Distribution Curve:NiagaraDataInterfaceCurve=DataInterface() {propertyValues {Curve {defaultValue=0, keys [{tangentMode=RCTM_User, time=0, value=0}, <at 0.5438 0.0594>, {tangentMode=RCTM_User, time=0.8658, value=0.78}]}}}
      Axis Variation:NiagaraFloat=0
      Randomness Mode:ENiagaraRandomnessMode=::NewEnumerator0
      Evaluation Type:ENiagaraRandomnessEvaluation=::NewEnumerator0
      Seed:NiagaraInt32=0
      Fixed Random Seed:NiagaraBool=false
     )
     2=&RandomRangeVector2D(Minimum=<10.9167 45.2909>, Maximum=<14.2053 86.2565>)
     3=?
     4=?
     Index=&SpawnGroup()
     Num=::NewEnumerator1
    )
    Sprite Size Mode=::NewEnumerator4
    UsePositionOffset=false
    Write Material Random=true
   )
   &CameraOffset(Camera Offset Amount=&RandomRangeFloat(Minimum=0, Maximum=20))
   &ShapeLocation(
    Handle Radius=11.1985
    Large Radius=&MergeHelper_FloatSwitch(0=45.6613, 1=45.6613, 2=40.8999, 3=0, 4=0, Index=&SpawnGroup(), Num=::NewEnumerator3)
    MaskSpawnGroup=false
    Offset=&MergeHelper_VectorSwitch(
     0=<0 0 -34.3861>
     1=<0 0 -34.3861>
     2=<0 0 -6.2569>
     3=<0 0 0>
     4=<0 0 0>
     Index=&SpawnGroup()
     Num=::NewEnumerator3
    )
    Offset Mode=::NewEnumerator0
    Override Seed=false
    Rotation Coordinate Space=::World
    Shape Origin=&MakeVector(X=0, Y=0, Z=0)
    Shape Primitive=::NewEnumerator3
    Surface Distribution=1
    UseOwnerScale=true
   )
   &AddVelocityFromPoint(
    Module.Use Falloff Distance=false
    Origin Offset=&MergeHelper_VectorSwitch(
     0=<0 0 -92.361>
     1=<0 0 -92.361>
     2=<0 0 -35.4048>
     3=<0 0 0>
     4=<0 0 0>
     Index=&SpawnGroup()
     Num=::NewEnumerator3
    )
    Velocity Origin:Vector3f=&MakeVector(X=0, Y=0, Z=0)
    Velocity Strength=&MergeHelper_FloatSwitch(
     0=&RandomRangeFloat(Minimum=310.1153, Maximum=589.3828)
     1=373.0592
     2=27.2009
     3=0
     4=0
     Index=&SpawnGroup()
     Num=::NewEnumerator3
    )
   )
  ]
  particleUpdateScript [
   &ParticleState(DeltaTime:NiagaraFloat=&Multiply_Float(A=$Engine.DeltaTime, B=$User.SpeedCurve))
   &Drag(Drag=5, Use Linear Drag=true, Use Rotational Drag=true)
   &SpriteRotationRate(
    Delta Time:NiagaraFloat=&Multiply_Float(A=$Engine.DeltaTime, B=$User.SpeedCurve)
    Rotation Rate=&RandomRangeFloat(Minimum=-200, Maximum=200)
   )
   &ScaleColor(
    Scale Alpha=&MergeHelper_FloatSwitch(
     0=&FloatFromCurve(
      FloatCurve=DataInterface() {propertyValues {Curve {
         defaultValue=1
         keys [
          {interpMode=RCIM_Cubic, tangentMode=RCTM_User, time=0.007, value=-0.0386}
          {arriveTangent=-2.8196, interpMode=RCIM_Cubic, leaveTangent=-2.8196, tangentMode=RCTM_User, time=0.5, value=1}
          {tangentMode=RCTM_User, time=1, value=0}
         ]
        }}}
      Scale Curve=2
     )
     1=&FloatFromCurve(FloatCurve=DataInterface() {propertyValues {Curve {
         defaultValue=1
         keys [
          {interpMode=RCIM_Cubic, tangentMode=RCTM_User, time=0.0247, value=1.0201}
          {arriveTangent=-0.3193, interpMode=RCIM_Cubic, leaveTangent=-0.3193, tangentMode=RCTM_User, time=0.1931, value=0.9292}
          {tangentMode=RCTM_User, time=1, value=0}
         ]
        }}})
     2=&FloatFromCurve(FloatCurve=DataInterface() {propertyValues {Curve {
         defaultValue=1
         keys [
          {interpMode=RCIM_Cubic, tangentMode=RCTM_User, time=0.0247, value=1.0201}
          {arriveTangent=-2.8196, interpMode=RCIM_Cubic, leaveTangent=-2.8196, tangentMode=RCTM_User, time=0.5, value=1}
          {tangentMode=RCTM_User, time=1, value=0}
         ]
        }}})
     3=0
     4=0
     Index=&SpawnGroup()
     Num=::NewEnumerator3
    )
    Scale RGB=&MakeVectorFromLinearColorRGB(Color=&RandomRangeLinearColor(
      Minimum=&ColorFromCurve(ColorCurve=DataInterface() {propertyValues {AlphaCurve {keys []}, BlueCurve {keys [<at 0.6529 1>]}, GreenCurve {keys [<at 0.6529 1>]}, RedCurve {keys [<at 0.6529 1>]}}})
      Maximum=<rgba 1 1 1 1>
     ))
    ScaleA=true
    ScaleRGB=true
   )
   &ScaleSpriteSize(Scale Factor=&MergeHelper_Vector2Switch(
     0=&Vector2DFromFloat(Value=&FloatFromCurve(
       FloatCurve=DataInterface() {propertyValues {Curve {keys [
           {arriveTangent=0.0418, interpMode=RCIM_Cubic, leaveTangent=0.0418, tangentMode=RCTM_User, time=0, value=0.3}
           {arriveTangent=0.8771, interpMode=RCIM_Cubic, leaveTangent=0.8771, tangentMode=RCTM_User, time=0.014, value=0.7752}
           {interpMode=RCIM_Cubic, leaveTangent=-0.7578, time=0.8047, value=1.4047}
          ]}}}
       Scale Curve=0.9
      ))
     1=&Vector2DFromCurve(
      Vector2Curve=DataInterface() {propertyValues {
        XCurve {defaultValue=1, keys [{arriveTangent=0.0224, interpMode=RCIM_Cubic, leaveTangent=0.0224, tangentMode=RCTM_User, time=1, value=1}]}
        YCurve {
         defaultValue=1
         keys [
          {arriveTangent=4.1785, interpMode=RCIM_Cubic, leaveTangent=4.1785, tangentMode=RCTM_User, time=0, value=0}
          {arriveTangent=1.1179, interpMode=RCIM_Cubic, leaveTangent=1.1179, tangentMode=RCTM_User, time=0.044, value=0.7113}
          {arriveTangent=0.0224, interpMode=RCIM_Cubic, leaveTangent=0.0224, tangentMode=RCTM_User, time=1, value=1}
         ]
        }
       }}
      ScaleCurve=<1 1.4198>
     )
     2=<0 0>
     3=<0 0>
     4=<0 0>
     Index=&SpawnGroup()
     Num=::NewEnumerator3
    ))
   &GravityForce(Gravity=&MergeHelper_VectorSwitch(
     0=&RandomRangeVector(Minimum=<0 0 352.5917>, Maximum=<1 1 2149.6328>)
     1=<0 0 0>
     2=<0 0 0>
     3=<0 0 0>
     4=<0 0 0>
     Index=&SpawnGroup()
     Num=::NewEnumerator3
    ))
   &InheritSourceMovement(ApplyForce=false, ApplyPosition=true, ApplyPositionFalloff=false)
   &SolveForcesAndVelocity(
    Clamp Velocity=false
    DeltaTime:NiagaraFloat=&Multiply_Float(A=$Engine.DeltaTime, B=$User.SpeedCurve)
    Limit Acceleration=false
   )
   &ScaleSpriteSizeBySpeed(Max Scale Factor=&MergeHelper_Vector2Switch(0=<1 1>, 1=<1 2.8028>, 2=<1 1.0916>, 3=<0 0>, 4=<0 0>, Index=&SpawnGroup(), Num=::NewEnumerator3))
   &ScaleColor(
    Scale Alpha=&FloatFromCurve(FloatCurve=DataInterface() {propertyValues {Curve {keys [<at 0 1>,<at 1 0>]}}})
    Scale RGB=&MakeVectorFromLinearColorRGB(Color=&MakeCustomLinearColorFromBool(
      Bool=$User.bIsFrontEndPreview
      True Color=&RandomRangeLinearColor(Minimum=<rgba 0.4 0.55 1 1>, Maximum=<rgba 0.08 0.11 0.2 1>)
      False Color=<rgba 1 1 1 1>
     ))
    ScaleA=true
    ScaleRGB=true
   ) {module=ScaleColor001}
   &DynamicMaterialParameters(
    Index 0 Param 1=&Add_Float(
     A=&RandomRangeFloat(Minimum=0, Maximum=1)
     B=&FloatFromCurve(
      FloatCurve=DataInterface() {propertyValues {Curve {keys [
          {arriveTangent=2.7533, interpMode=RCIM_Cubic, leaveTangent=1.8958, tangentMode=RCTM_User, time=0, value=1}
          {arriveTangent=0.8292, interpMode=RCIM_Cubic, leaveTangent=0.8292, tangentMode=RCTM_User, time=0.321, value=1.3453}
          {arriveTangent=0.0529, interpMode=RCIM_Cubic, leaveTangent=0.2021, tangentMode=RCTM_User, time=1, value=1.5}
         ]}}}
      CurveIndex=&Add_Float(A=&Multiply_Float(A=$Particles.Initial.MaterialRandom, B=0.5), B=$Particles.NormalizedAge)
     )
    )
    Index 0 Param 2=&FloatFromCurve(FloatCurve=DataInterface() {propertyValues {Curve {keys [
         {arriveTangent=0.0453, interpMode=RCIM_Cubic, tangentMode=RCTM_Break, time=0, value=-1}
         {arriveTangent=0.0604, interpMode=RCIM_Cubic, leaveTangent=0.0604, tangentMode=RCTM_User, time=1, value=-1}
        ]}}})
    Index 0 Param 3=&FloatFromCurve(
     FloatCurve=DataInterface() {propertyValues {Curve {keys [
         {arriveTangent=0.0632, interpMode=RCIM_Cubic, leaveTangent=0.0632, tangentMode=RCTM_User, time=0, value=20}
         {
          arriveTangent=-1.6858
          arriveTangentWeight=0.0333
          interpMode=RCIM_Cubic
          leaveTangent=-1.6858
          leaveTangentWeight=0.3137
          tangentMode=RCTM_User
          time=0.05
          value=0.1
         }
         {arriveTangentWeight=0.1027, interpMode=RCIM_Cubic, tangentMode=RCTM_Break, time=0.2, value=0}
        ]}}}
     Scale Curve=&RandomRangeFloat(Minimum=0.25, Maximum=0.5)
    )
    Index 0 Param 4=&FloatFromCurve(FloatCurve=DataInterface() {propertyValues {Curve {keys [<at 0.5 1>,<at 1 0>]}}})
    Index 1 Param 1=&FloatFromCurve(
     FloatCurve=DataInterface() {propertyValues {Curve {keys [
         {interpMode=RCIM_Cubic, tangentMode=RCTM_Break, time=0, value=1}
         {arriveTangentWeight=0.0833, interpMode=RCIM_Cubic, tangentMode=RCTM_Break, time=0.25, value=0.3}
        ]}}}
     Scale Curve=&RandomRangeFloat(Minimum=1.5, Maximum=3)
    )
    Index 1 Param 2=&FloatFromCurve(
     FloatCurve=DataInterface() {propertyValues {Curve {keys [
         {interpMode=RCIM_Cubic, leaveTangent=-2.5489, tangentMode=RCTM_Break, time=0, value=1}
         {interpMode=RCIM_Cubic, leaveTangent=0.9066, tangentMode=RCTM_Break, time=1, value=0}
        ]}}}
     Scale Curve=&RandomRangeFloat(Minimum=0.01, Maximum=0.1)
    )
    Index 1 Param 3=&FloatFromCurve(FloatCurve=DataInterface() {propertyValues {Curve {keys [
         {interpMode=RCIM_Cubic, leaveTangent=6.3917, tangentMode=RCTM_Break, time=0, value=0.2}
         {arriveTangent=1.7549, interpMode=RCIM_Cubic, leaveTangent=10.6767, tangentMode=RCTM_User, time=0.25, value=1}
        ]}}})
    Index 1 Param 4=&FloatFromCurve(
     FloatCurve=DataInterface() {propertyValues {Curve {keys [{arriveTangent=-0.136, interpMode=RCIM_Cubic, leaveTangent=-0.136, tangentMode=RCTM_User, time=0, value=1}, <at 1 0.95>]}}}
     Scale Curve=&RandomRangeFloat(Minimum=1.3, Maximum=2)
    )
    Index 2 Param 1=&FloatFromCurve(FloatCurve=DataInterface() {propertyValues {Curve {keys [
         {interpMode=RCIM_Cubic, leaveTangent=4.2485, tangentMode=RCTM_Break, time=0, value=0}
         {arriveTangent=1.0548, interpMode=RCIM_Cubic, leaveTangent=1.0548, tangentMode=RCTM_User, time=0.17, value=0.5715}
         <at 1 1>
        ]}}})
    Index 2 Param 2=&FloatFromCurve(
     FloatCurve=DataInterface() {propertyValues {Curve {keys [
         {arriveTangent=1.9056, interpMode=RCIM_Cubic, leaveTangent=1.9056, tangentMode=RCTM_User, time=0, value=0.2}
         {arriveTangent=0.0259, interpMode=RCIM_Cubic, leaveTangent=0.0259, tangentMode=RCTM_User, time=1, value=1.4}
        ]}}}
     Scale Curve=&RandomRangeFloat(Minimum=0.1, Maximum=0.2)
    )
    Index 2 Param 3=&FloatFromCurve(
     FloatCurve=DataInterface() {propertyValues {Curve {keys [
         {interpMode=RCIM_Cubic, leaveTangent=0.0393, tangentMode=RCTM_Break, time=0, value=1}
         {interpMode=RCIM_Cubic, leaveTangent=11.1233, tangentMode=RCTM_Break, time=1, value=1.5}
        ]}}}
     Scale Curve=0.7
    )
    Index 2 Param 4=&FloatFromCurve(
     FloatCurve=DataInterface() {propertyValues {Curve {keys [
         {arriveTangent=-0.6867, interpMode=RCIM_Cubic, leaveTangent=-0.6867, tangentMode=RCTM_User, time=0, value=1}
         {interpMode=RCIM_Cubic, leaveTangent=-1.2956, tangentMode=RCTM_Break, time=1, value=0.75}
        ]}}}
     Scale Curve=1.5
    )
    Module.Param0WriteEnabled=true
    Module.Param0WriteEnabled1=true
    Module.Param0WriteEnabled2=true
    Module.Param1WriteEnabled=false
    Module.Param1WriteEnabled1=true
    Module.Param1WriteEnabled2=true
    Module.Param2WriteEnabled=true
    Module.Param2WriteEnabled1=true
    Module.Param2WriteEnabled2=true
    Module.Param3WriteEnabled=true
    Module.Param3WriteEnabled1=true
    Module.Param3WriteEnabled2=true
    Write Parameter Index 1=true
    Write Parameter Index 2=true
   ) {module=DynamicMaterialParameters001}
   /SetVariables(Particles.Position2:NiagaraPosition=&AddVectorToPosition(Position=$Engine.Emitter.SimulationPosition, Vector=Linked() {linkedVariable {name=Particles.Position, type=NiagaraPosition}})) {isSetParameters=true}
   /SetVariables(Particles.Vector 2D:Vector2f=&Multiply_Vector2D(
     A=<1 1>
     B=&Vector2DFromCurve(
      Vector2Curve=DataInterface() {propertyValues {
        XCurve {defaultValue=1, keys [{arriveTangent=0.0224, interpMode=RCIM_Cubic, leaveTangent=0.0224, tangentMode=RCTM_User, time=1, value=1}]}
        YCurve {
         defaultValue=1
         keys [
          {arriveTangent=4.1785, interpMode=RCIM_Cubic, leaveTangent=4.1785, tangentMode=RCTM_User, time=0, value=0}
          {arriveTangent=1.1179, interpMode=RCIM_Cubic, leaveTangent=1.1179, tangentMode=RCTM_User, time=0.1111, value=0.6425}
          {arriveTangent=0.0224, interpMode=RCIM_Cubic, leaveTangent=0.0224, tangentMode=RCTM_User, time=1, value=1}
         ]
        }
       }}
      ScaleCurve=&RandomRangeVector2D(Minimum=<5 66.0044>, Maximum=<17.7984 123.9738>)
     )
    )) {isSetParameters=true}
  ]
  rendererClasses [&NiagaraSpriteRendererProperties]
  renderers [{rendererClass=&NiagaraSpriteRendererProperties}]
 }
}
summary {
 systemName=NS_Emote_HeroicTale_GolemExplosion
 emitters [
  {emitterName=Chunks, rendererClasses [&NiagaraMeshRendererProperties]}
  {emitterName=Dust, rendererClasses [&NiagaraSpriteRendererProperties]}
  {emitterName=Glow, rendererClasses [&NiagaraSpriteRendererProperties, &NiagaraLightRendererProperties]}
  {emitterName=Smokelow, rendererClasses [&NiagaraSpriteRendererProperties]}
  {emitterName=GroundCrack, rendererClasses [&NiagaraMeshRendererProperties]}
  {emitterName=Flare_Complex, rendererClasses [&NiagaraMeshRendererProperties]}
  {emitterName=Flash, rendererClasses [&NiagaraMeshRendererProperties]}
  {emitterName=Low_Med, rendererClasses [&NiagaraDecalRendererProperties]}
 ]
 userVariables [
  User.bIsFrontEnd {type=NiagaraBool}
  User.bIsFrontEndPreview {type=NiagaraBool}
  User.SpeedCurve {type=NiagaraFloat, defaultValue=1, description=""}
 ]
}
