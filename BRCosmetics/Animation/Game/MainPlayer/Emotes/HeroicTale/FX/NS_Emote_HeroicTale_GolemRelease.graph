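# Niagara burst effect for the golem release moment of the HeroicTale player emote, in orange and gold.
# Spikes fires two delayed bursts from a ring, using spawn groups to switch between short-lived streaks aligned to the shape normal and fast, noise-driven dots.
# BlowingParticles001 adds rising, curl-noise embers after a short delay; Smoke and Smoke2 list no renderers.
# A system time-of-day value scales streak alpha, and a front-end modifier adjusts dot scale.
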NiagaraSystem /BRCosmetics/Animation/Game/MainPlayer/Emotes/HeroicTale/FX/NS_Emote_HeroicTale_GolemRelease
dependencies {
 usedDataInterfaces [&NiagaraDataInterfaceArrayColor, &NiagaraDataInterfaceCurve, &NiagaraDataInterfaceVector2DCurve]
 usedDynamicInputs [&FNTimeOfDayScalar]
 usedModules [
  &self:SystemSpawnScript.SystemScriptSource.SystemScriptGraph.NiagaraNodeAssignment_0.SetVariables_#1
  &SystemState
  &self:SystemSpawnScript.SystemScriptSource.SystemScriptGraph.NiagaraNodeAssignment_1.SetVariables_#2
 ]
}
emitters {
 BlowingParticles001 {
  emitterUpdateScript [
   &EmitterState(Life Cycle Mode=::NewEnumerator1, Loop Behavior=::NewEnumerator1, Loop Delay=0.2, Loop Duration=1, UseLoopDelay=true)
   &SpawnRate(SpawnRate=53.8872, Use Spawn Probability=false)
  ]
  particleSpawnScript [
   &InitializeParticle(
    Color=<rgba 3 0.7206 0 1>
    Lifetime Max=1
    Lifetime Min=0.5
    Lifetime Mode=::NewEnumerator1
    Mass Max=1.5
    Mass Mode=::NewEnumerator3
    Sprite Rotation Mode=::NewEnumerator2
    Sprite Size Mode=::NewEnumerator1
    Uniform Sprite Size Max=2
    Uniform Sprite Size Min=1
    UsePositionOffset=false
    Write Material Random=true
   )
   &ShapeLocation(MaskSpawnGroup=false, Override Seed=false, Sphere Radius=55.7887, UseOwnerScale=true)
  ]
  particleUpdateScript [
   &ParticleState()
   &GravityForce(Gravity=&RandomRangeVector(Minimum=<0 38.7989 100>, Maximum=<0 173.5837 500>))
   &CurlNoiseForce(Bias Noise=false, Noise Strength=580.2303, Pan Noise=false)
   &Drag(
    Drag=&FloatFromCurve(FloatCurve=DataInterface() {propertyValues {Curve {
        defaultValue=1
        keys [
         {arriveTangent=4.1785, interpMode=RCIM_Cubic, leaveTangent=4.1785, tangentMode=RCTM_User, time=0, value=0}
         {arriveTangent=1.1179, interpMode=RCIM_Cubic, leaveTangent=1.1179, tangentMode=RCTM_User, time=0.2042, value=0.6228}
         {arriveTangent=0.0224, interpMode=RCIM_Cubic, leaveTangent=0.0224, tangentMode=RCTM_User, time=1, value=1}
        ]
       }}})
    Use Linear Drag=true
    Use Rotational Drag=true
   )
   &SolveForcesAndVelocity(Clamp Velocity=false, Limit Acceleration=false)
   &ScaleSpriteSizeBySpeed(Max Scale Factor=<1 4>, Min Scale Factor=<1 2>)
   &ScaleSpriteSize(Uniform Curve Sprite Scale=DataInterface() {propertyValues {Curve {keys [
        {arriveTangent=14.5835, interpMode=RCIM_Cubic, leaveTangent=14.5835, tangentMode=RCTM_User, time=0, value=0}
        {arriveTangent=1.1111, interpMode=RCIM_Cubic, leaveTangent=1.1111, time=0.1, value=1}
        {arriveTangent=-1.1111, interpMode=RCIM_Cubic, leaveTangent=-1.1111, time=0.9, value=1}
        {arriveTangent=-8.7113, interpMode=RCIM_Cubic, leaveTangent=-8.7113, tangentMode=RCTM_User, time=1, value=0}
       ]}}})
  ]
  rendererClasses [&NiagaraSpriteRendererProperties]
  renderers [{rendererClass=&NiagaraSpriteRendererProperties}]
 }
 Smoke {rendererClasses [], renderers []}
 Smoke2 {rendererClasses [], renderers []}
 Spikes {
  emitterUpdateScript [
   &EmitterState(Life Cycle Mode=::NewEnumerator1, Loop Behavior=::NewEnumerator1, Loop Duration=1, UseLoopDelay=true)
   &SpawnBurst_Instantaneous(Spawn Count=39, Spawn Time=$System.BurstDelay, Use Spawn Probability=false, UseLoopCountLimit=false)
   &SpawnBurst_Instantaneous(Spawn Count=41, Spawn Group=1, Spawn Time=$System.BurstDelay, Use Spawn Probability=false, UseLoopCountLimit=false) {module=SpawnBurst_Instantaneous001}
  ]
  particleSpawnScript [
   /SetVariables(Particles.VisibilityTag=&SpawnGroup()) {isSetParameters=true}
   &InitializeParticle(
    Color=&SelectLinearColorFromArray(
     Array Sampling Mode=::NewEnumerator0
     Direct Array Index=?
     Color Selection Array=DataInterface() {propertyValues {ColorData [<rgba 5 3.0422 0 0.1652>, <rgba 1.667 0.409 0 0.1652>]}}
     Random Seed=0
    )
    Lifetime=&MergeHelper_FloatSwitch(
     0=&RandomRangeFloat(Minimum=0.16, Maximum=0.2968)
     1=&RandomRangeFloat(Minimum=0.46, Maximum=0.9)
     2=?
     3=?
     Index=&SpawnGroup()
     Num=::NewEnumerator0
    )
    Sprite Size=&MergeHelper_Vector2Switch(
     0=&RandomRangeVector2D(Minimum=<305.6159 175.1617>, Maximum=<150.0497 181.2326>)
     1=&Multiply_Vector2D_ByFloat(Float=&RandomRangeFloat(Minimum=4.961, Maximum=6.1472), Vector2D=<1 1>)
     2=<0 0>
     3=<0 0>
     4=<0 0>
     Index=&SpawnGroup()
     Num=::NewEnumerator3
    )
    Sprite Size Mode=::NewEnumerator4
    UsePositionOffset=true
    Write Material Random=true
   )
   &SubUVAnimation(End Frame=0, SubUV Animation Mode:ENiagara_SubUVLookupMode=::NewEnumerator2)
   &ShapeLocation(
    MaskSpawnGroup=false
    Offset Coordinate Space=::Local
    Offset Mode=::NewEnumerator0
    Override Seed=false
    Ring Radius=41.386
    Rotation Mode=::NewEnumerator2
    Shape Primitive=::NewEnumerator4
    UseOwnerScale=true
    Yaw / Pitch / Roll=<0.25 0 0>
   )
   &AddVelocity(
    Module.Use Falloff Distance=false
    Origin Offset Coordinate Space=::Simulation
    Override Seed=false
    Velocity Mode=::NewEnumerator1
    Velocity Speed=&MergeHelper_FloatSwitch(
     0=0.096
     1=&RandomRangeFloat(Minimum=221.0147, Maximum=717.2778)
     2=?
     3=?
     Index=&SpawnGroup()
     Num=::NewEnumerator0
    )
   )
   &CameraOffset(Camera Offset Amount=&RandomRangeFloat(Minimum=0, Maximum=15))
  ]
  particleUpdateScript [
   &ParticleState()
   &Drag(
    Drag=&MergeHelper_FloatSwitch(
     0=0
     1=&RandomRangeFloat(Minimum=2.5633, Maximum=7.3245)
     2=?
     3=?
     Index=&SpawnGroup()
     Num=::NewEnumerator0
    )
    Use Linear Drag=true
    Use Rotational Drag=false
   )
   &CurlNoiseForce(
    Bias Noise=false
    Noise Strength=&MergeHelper_FloatSwitch(0=0, 1=264.451, 2=0, 3=0, 4=0, Index=&SpawnGroup(), Num=::NewEnumerator3)
    Pan Noise=false
   )
   &SolveForcesAndVelocity(Clamp Velocity=false, Limit Acceleration=false)
   &ScaleSpriteSizeBySpeed(
    Max Scale Factor=&MergeHelper_Vector2Switch(0=<1 1>, 1=<1 1.5842>, 2=?, 3=?, 4=?, Index=&SpawnGroup(), Num=::NewEnumerator0)
    Min Scale Factor=&MergeHelper_Vector2Switch(0=<1 1>, 1=<1 2.3119>, 2=?, 3=?, 4=?, Index=&SpawnGroup(), Num=::NewEnumerator0)
   )
   /SetVariables(Particles.DotScale:Vector2f=&Vector2DFromCurve(
     Vector2Curve=DataInterface() {propertyValues {
       XCurve {keys [{tangentMode=RCTM_User, time=0, value=1}, <at 0.7653 0.9082>, {tangentMode=RCTM_User, time=1, value=0}]}
       YCurve {keys [{tangentMode=RCTM_User, time=0, value=0.5}, <at 0.5931 1.7242>, {tangentMode=RCTM_User, time=0.9916, value=0}]}
      }}
     ScaleCurve=&Multiply_Vector2D_ByFloat(
      Float=&FN_FrontEndFloatModifier(Gameplay=0.6, Front End=0.5)
      Vector2D=&Multiply_Vector2D(A=<0.5 0.5>, B=$Particles.SpriteSize)
     )
    )) {isSetParameters=true}
   &ScaleSpriteSize(
    Scale Factor=&MergeHelper_Vector2Switch(
     0=&Vector2DFromCurve(
      Vector2Curve=DataInterface() {propertyValues {
        XCurve {keys [{tangentMode=RCTM_User, time=0, value=1}, {tangentMode=RCTM_User, time=1, value=1}]}
        YCurve {
         defaultValue=0.1635
         keys [
          {arriveTangent=4.1785, interpMode=RCIM_Cubic, leaveTangent=4.1785, tangentMode=RCTM_User, time=0, value=0}
          {arriveTangent=1.1179, interpMode=RCIM_Cubic, leaveTangent=1.1179, tangentMode=RCTM_User, time=0.2042, value=0.6228}
          {arriveTangent=0.0224, interpMode=RCIM_Cubic, leaveTangent=0.0224, tangentMode=RCTM_User, time=1, value=1}
         ]
        }
       }}
      ScaleCurve=<0.424 1.4199>
     )
     1=&Vector2DFromCurve(
      Vector2Curve=DataInterface() {propertyValues {
        XCurve {keys [{tangentMode=RCTM_User, time=0, value=1}, {tangentMode=RCTM_User, time=1, value=1}]}
        YCurve {keys [{tangentMode=RCTM_User, time=0, value=0.0199}, {tangentMode=RCTM_User, time=1, value=1.7}]}
       }}
      ScaleCurve=<1 1>
     )
     2=<0 0>
     3=<0 0>
     4=<0 0>
     Index=0
     Num=::NewEnumerator3
    )
    Scale Sprite Size Mode=::NewEnumerator2
   )
   &ScaleColor(
    Scale Alpha=&MergeHelper_FloatSwitch(
     0=&FloatFromCurve(FloatCurve=DataInterface() {propertyValues {Curve {keys [<at 0 0>,<at 0.1 1>,<at 0.2 0.8>,<at 1 0>]}}}, Scale Curve=$System.TOD)
     1=1
     2=?
     3=?
     Index=&SpawnGroup()
     Num=::NewEnumerator0
    )
    ScaleA=true
    ScaleRGB=true
   )
   &SpriteFacingAndAlignment(
    Sprite Alignment=$Particles.ShapeLocation.ShapeNormal
    Sprite Facing:NiagaraPosition=$Particles.GimbalPos
    WriteAlignment=true
    WriteFacing=false
   )
  ]
  rendererClasses [&NiagaraSpriteRendererProperties]
  renderers [
   {rendererClass=&NiagaraSpriteRendererProperties}
   {rendererClass=&NiagaraSpriteRendererProperties, rendererIndex=1}
  ]
 }
}
summary {
 systemName=NS_Emote_HeroicTale_GolemRelease
 emitters [
  {emitterName=Spikes, rendererClasses [&NiagaraSpriteRendererProperties]}
  {emitterName=Smoke, rendererClasses []}
  {emitterName=Smoke2, rendererClasses []}
  {emitterName=BlowingParticles001, rendererClasses [&NiagaraSpriteRendererProperties]}
 ]
 userVariables [
  User.bIsFrontEnd {type=NiagaraBool}
  User.bIsFrontEndPreview {type=NiagaraBool}
 ]
}
