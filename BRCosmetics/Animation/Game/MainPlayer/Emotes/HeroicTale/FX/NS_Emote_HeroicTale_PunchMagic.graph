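# Niagara system for the Heroic Tale emote's magic punch, a cyan-blue impact effect.
# Hand glow sprites and a ribbon trail follow the hand_r bone through skeletal mesh sampling.
# At a shared system burst delay, Flare002 fires flashes, a mesh and a light while Smokelow bursts smoke sprites and meshes with drag and curl noise.
# SpeedCurve scales smoke simulation time, and front-end inputs adjust flare size and smoke tint.
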NiagaraSystem /BRCosmetics/Animation/Game/MainPlayer/Emotes/HeroicTale/FX/NS_Emote_HeroicTale_PunchMagic
dependencies {
 usedDataInterfaces [
  &NiagaraDataInterfaceSkeletalMesh
  &NiagaraDataInterfaceCurve
  &NiagaraDataInterfaceColorCurve
  &NiagaraDataInterfaceVector2DCurve
 ]
 usedDynamicInputs [&MakeFloatFromVector, &FNTimeOfDayScalar]
 usedModules [
  &self:SystemSpawnScript.SystemScriptSource.SystemScriptGraph.NiagaraNodeAssignment_0.SetVariables_#1
  &SystemState
  &self:SystemSpawnScript.SystemScriptSource.SystemScriptGraph.NiagaraNodeAssignment_1.SetVariables_#2
 ]
}
emitters {
 Flare002 {
  emitterUpdateScript [
   &EmitterState(Life Cycle Mode=::NewEnumerator1, Loop Behavior=::NewEnumerator1, Loop Duration=1.5377, UseLoopDelay=true)
   &SpawnBurst_Instantaneous(Spawn Count=1, Spawn Time=$System.BurstDelay, Use Spawn Probability=false)
   &SpawnBurst_Instantaneous(Spawn Count=1, Spawn Group=3, Spawn Time=$System.BurstDelay, Use Spawn Probability=false) {module=SpawnBurst_Instantaneous001}
   &SpawnBurst_Instantaneous(Spawn Count=1, Spawn Group=1, Spawn Time=$System.BurstDelay, Use Spawn Probability=false) {module=SpawnBurst_Instantaneous002}
  ]
  particleSpawnScript [
   /SetVariables(Particles.VisibilityTag=&SpawnGroup()) {isSetParameters=true}
   &InitializeParticle(
    Color=<rgba 0 0.6985 1.78 1>
    Lifetime=&MergeHelper_FloatSwitch(0=0.15, 1=0.748, 2=1.5231, 3=?, Index=&SpawnGroup(), Num=::NewEnumerator1)
    Mesh Scale Mode=::NewEnumerator3
    Mesh Uniform Scale=2.0439
    Sprite Rotation Mode=::NewEnumerator2
    Sprite Size Mode=::NewEnumerator3
    Uniform Sprite Size=269.9966
    UsePositionOffset=true
    Write Material Random=true
   )
   &DynamicMaterialParameters(
    Index 0 Param 1=&RandomRangeFloat(Minimum=0.4, Maximum=0.6)
    Index 0 Param 2=1
    Index 0 Param 3=1
    Index 0 Param 4=1
    Module.Param0WriteEnabled=true
    Module.Param1WriteEnabled=true
    Module.Param2WriteEnabled=true
    Module.Param3WriteEnabled=true
   )
   /SetVariables(Particles.FlareRot=91.5131) {isSetParameters=true}
  ]
  particleUpdateScript [
   &ParticleState()
   &ScaleMeshSize(Scale Factor=&VectorFromFloat(Value=&FloatFromCurve(FloatCurve=DataInterface() {propertyValues {Curve {
         defaultValue=1
         keys [
          {arriveTangent=1.6237, interpMode=RCIM_Cubic, leaveTangent=1.6237, tangentMode=RCTM_User, time=0.2042, value=1}
          {arriveTangent=0.5742, interpMode=RCIM_Cubic, leaveTangent=0.5742, time=0.3314, value=1.3207}
          {arriveTangent=0.0224, interpMode=RCIM_Cubic, leaveTangent=0.0224, tangentMode=RCTM_User, time=0.9794, value=1.4451}
         ]
        }}})))
   &SolveForcesAndVelocity(Clamp Velocity=false, Limit Acceleration=false)
   &ScaleColor(
    Scale Alpha=&Multiply_Float(
     A=$System.TOD
     B=&FloatFromCurve(FloatCurve=DataInterface() {propertyValues {Curve {keys [<at 0 0>,<at 0.2 1>,<at 0.8 1>,<at 1 0>]}}})
    )
    ScaleA=true
    ScaleRGB=true
   )
   &ScaleSpriteSize(Scale Factor=&Vector2DFromFloat(Value=&FloatFromCurve(
      FloatCurve=DataInterface() {propertyValues {Curve {keys [<at 0 1.8233 RCIM_Cubic>,<at 1 0>]}}}
      Scale Curve=&FN_FrontEndFloatModifier(Gameplay=1, Front End=0.7)
     )))
   &CameraOffset(Camera Offset Amount=51.3705)
   /SetVariables(Particles.Flash size:Vector2f=&Multiply_Vector2D_ByFloat(
     Float=&Multiply_Float(
      A=353.8598
      B=&FloatFromCurve(
       FloatCurve=DataInterface() {propertyValues {Curve {keys [{arriveTangent=-0.0405, interpMode=RCIM_Cubic, leaveTangent=-0.0405, tangentMode=RCTM_User, time=0.62, value=1.7699}]}}}
       Scale Curve=&FN_FrontEndFloatModifier(Gameplay=1, Front End=0.7)
      )
     )
     Vector2D=<1 1>
    )) {isSetParameters=true}
   /SetVariables(Particles.Flash color:LinearColor=&ColorFromCurve(ColorCurve=DataInterface() {propertyValues {
       AlphaCurve {keys [<at 0.1154 0.694>, {arriveTangent=-0.7846, interpMode=RCIM_Cubic, leaveTangent=-0.7846, time=0.7197, value=0.7}, <at 1 0 RCIM_Cubic>]}
       BlueCurve {keys [<at 0 4.631 RCIM_Cubic>,<at 0.4266 3.2292>]}
       GreenCurve {keys [<at 0 3.5009 RCIM_Cubic>,<at 0.4266 0.9058>]}
       RedCurve {keys [<at 0 0.5548 RCIM_Cubic>,<at 0.4266 0>]}
      }})) {isSetParameters=true}
   /SetVariables(Particles.flash2size:Vector2f=&Multiply_Vector2D_ByFloat(
     Float=&Multiply_Float(
      A=292.5097
      B=&FloatFromCurve(
       FloatCurve=DataInterface() {propertyValues {Curve {keys [{arriveTangent=-0.0405, interpMode=RCIM_Cubic, leaveTangent=-0.0405, tangentMode=RCTM_User, time=0.62, value=1.7699}]}}}
       Scale Curve=&FN_FrontEndFloatModifier(Gameplay=1, Front End=0.7)
      )
     )
     Vector2D=<1 1>
    )) {isSetParameters=true}
   /SetVariables(Particles.flash2color:LinearColor=&MakeLinearColorFromVectorAndFloat(
     "Vector (RGB)"=&MakeVectorFromLinearColorRGB(Color=<rgba 0 0.3438 1 1>)
     "Float (Alpha)"=&Multiply_Float(
      A=&Multiply_Float(
       A=$System.TOD
       B=&FloatFromCurve(FloatCurve=DataInterface() {propertyValues {Curve {keys [<at 0 0>,<at 0.124 1>,<at 0.496 1>,<at 0.62 0>]}}})
      )
      B=1
     )
    )) {isSetParameters=true}
   /SetVariables(
    Particles.DyunmParamWave:Vector4f=&MakeVector4(
     X=&FloatFromCurve(FloatCurve=DataInterface() {propertyValues {Curve {keys [
          {interpMode=RCIM_Cubic, leaveTangent=-1.5024, tangentMode=RCTM_Break, time=0.18, value=0.5}
          {arriveTangent=-0.6098, interpMode=RCIM_Cubic, leaveTangent=-0.6098, time=0.3433, value=0.1559}
          <at 1 0>
         ]}}})
     Y=&FloatFromCurve(FloatCurve=DataInterface() {propertyValues {Curve {keys [{arriveTangent=-0.3464, interpMode=RCIM_Cubic, leaveTangent=-0.3464, tangentMode=RCTM_User, time=-0.0031, value=0.9929}, <at 0.87 0.9>]}}})
     Z=&FloatFromCurve(FloatCurve=DataInterface() {propertyValues {Curve {keys [<at 0 0>,<at 0.15 1.0141>,<at 1 1.2>]}}})
     W=1
    )
    Particles.WaveColor:LinearColor=&ColorFromCurve(ColorCurve=DataInterface() {propertyValues {
       AlphaCurve {keys [<at 0 1 RCIM_Cubic>,<at 0.1959 0.1721>,<at 0.9877 0 RCIM_Cubic>]}
       BlueCurve {keys [<at 0 2 RCIM_Cubic>,<at 0.1799 0.7396>]}
       GreenCurve {keys [<at 0 1.6736 RCIM_Cubic>,<at 0.1799 0.3796>]}
       RedCurve {keys [<at 0 0 RCIM_Cubic>,<at 0.1799 0>]}
      }})
   ) {isSetParameters=true}
   &FN_Light_Attributes(
    Color=<rgba 0 39.3082 60 1>
    Create Color Attribute=true
    Create Radius Attribute=true
    Radius=148.7189
    Scale Color=&FloatFromCurve(FloatCurve=DataInterface() {propertyValues {Curve {keys [<at 0.0605 1>,<at 0.1534 0.1676>,<at 1 0>]}}})
   )
   /SetVariables(Particles.SoftFlareColor:LinearColor=&MultiplyLinearColor(Linear Color=<rgba 1 1 1 0.3217>, Scale Factor=&MakeVector4_FromLinearColor(LinearColor=$Particles.Color))) {isSetParameters=true}
  ]
  rendererClasses [
   &NiagaraSpriteRendererProperties
   &NiagaraMeshRendererProperties
   &NiagaraLightRendererProperties
  ]
  renderers [
   {rendererClass=&NiagaraSpriteRendererProperties}
   {rendererClass=&NiagaraSpriteRendererProperties, rendererIndex=1}
   {rendererClass=&NiagaraSpriteRendererProperties, rendererIndex=2}
   {rendererClass=&NiagaraMeshRendererProperties, rendererIndex=3}
   {rendererClass=&NiagaraLightRendererProperties, rendererIndex=4}
   {rendererClass=&NiagaraSpriteRendererProperties, rendererIndex=5}
  ]
 }
 HandFX {
  emitterUpdateScript [
   &EmitterState(Life Cycle Mode=::NewEnumerator1, UseLoopDelay=false)
   &SpawnBurst_Instantaneous(Spawn Group=1, Use Spawn Probability=false, UseLoopCountLimit=false)
   &SpawnBurst_Instantaneous(Use Spawn Probability=false, UseLoopCountLimit=false) {module=SpawnBurst_Instantaneous001}
  ]
  particleSpawnScript [
   /SetVariables(Particles.VisibilityTag=&SpawnGroup()) {isSetParameters=true}
   &InitializeParticle(
    Color=&MergeHelper_ColorSwitch(
     0=<rgba 3.125 8.0939 10 1>
     1=<rgba 0 5.1589 10 0.1826>
     2=<rgba 0 0 0 0>
     3=<rgba 0 0 0 0>
     4=<rgba 0 0 0 0>
     Index=&SpawnGroup()
     Num=::NewEnumerator3
    )
    Lifetime=&MergeHelper_FloatSwitch(0=1.3864, 1=1.3, 2=0, 3=0, 4=0, Index=&SpawnGroup(), Num=::NewEnumerator3)
    Mass Max=1.5
    Mass Mode=::NewEnumerator3
    Sprite Rotation Mode=::NewEnumerator2
    Sprite Size Mode=::NewEnumerator3
    Uniform Sprite Size=&MergeHelper_FloatSwitch(0=129.6553, 1=78.6088, 2=0, 3=0, 4=0, Index=&SpawnGroup(), Num=::NewEnumerator3)
    UsePositionOffset=false
    Write Material Random=true
   )
   &ShapeLocation(MaskSpawnGroup=false, Override Seed=false, Sphere Radius=0, UseOwnerScale=true)
   &InitialMeshOrientation(Use Rotation Vector=false)
   &CameraOffset(Camera Offset Amount=20)
  ]
  particleUpdateScript [
   &ParticleState()
   &SkeletalMeshLocation(
    Bone Sampling Mode=::NewEnumerator4
    MaskSpawnGroup=false
    Skeletal Mesh=DataInterface() {propertyValues {
      DefaultMesh=null
      ExcludeBoneName=root
      OptionalComponentClass=null
      PreviewMesh=/Game/Characters/Player/Male/Male_Avg_Base/Fortnite_M_Avg_Player
      SkinningMode=SkinOnTheFly
      SoftSourceActor=null
      SourceMode=Default
      UvSetIndex=0
      WholeMeshLOD=-1
      bExcludeBone=false
      bReadDeformedGeometry=true
      bRequireCurrentFrameData=true
      ComponentTags []
      FilteredBones [hand_r]
      FilteredSockets []
      MeshUserParameter {parameter {name=None, typeDefHandle {registeredTypeIndex=1}, varData []}}
      SamplingRegions []
     }}
   )
   &SpriteRotationRate(Rotation Rate=&MergeHelper_FloatSwitch(
     0=&FloatFromCurve(
      FloatCurve=DataInterface() {propertyValues {Curve {
         defaultValue=1
         keys [
          {arriveTangent=9.0725, interpMode=RCIM_Cubic, leaveTangent=9.0725, tangentMode=RCTM_User, time=0, value=0}
          {arriveTangent=-2.5249, interpMode=RCIM_Cubic, leaveTangent=-2.5249, tangentMode=RCTM_User, time=0.1408, value=0.9847}
          {arriveTangent=-0.0405, interpMode=RCIM_Cubic, leaveTangent=-0.0405, tangentMode=RCTM_User, time=0.4512, value=0.0384}
         ]
        }}}
      Scale Curve=780
     )
     1=0
     2=0
     3=0
     4=0
     Index=&SpawnGroup()
     Num=::NewEnumerator3
    ))
   &SolveForcesAndVelocity(Clamp Velocity=false, Limit Acceleration=false)
   &ScaleSpriteSize(
    Scale Factor=&MergeHelper_Vector2Switch(
     0=&Multiply_Vector2D_ByFloat(
      Float=&FloatFromCurve(FloatCurve=DataInterface() {propertyValues {Curve {
          defaultValue=1
          keys [
           {interpMode=RCIM_Cubic, tangentMode=RCTM_User, time=0, value=0}
           {interpMode=RCIM_Cubic, tangentMode=RCTM_User, time=0.1308, value=1.1359}
           {arriveTangent=-1.3069, interpMode=RCIM_Cubic, leaveTangent=-1.3069, time=0.8542, value=1.0388}
           {tangentMode=RCTM_User, time=1, value=0}
          ]
         }}})
      Vector2D=<1 1>
     )
     1=<1 1>
     2=<0 0>
     3=<0 0>
     4=<0 0>
     Index=&SpawnGroup()
     Num=::NewEnumerator3
    )
    Scale Sprite Size Mode=::NewEnumerator2
   )
  ]
  rendererClasses [&NiagaraSpriteRendererProperties]
  renderers [
   {rendererClass=&NiagaraSpriteRendererProperties}
   {rendererClass=&NiagaraSpriteRendererProperties, rendererIndex=1}
   {rendererClass=&NiagaraSpriteRendererProperties, rendererIndex=2}
  ]
 }
 HandRibbon_ {
  emitterUpdateScript [
   &EmitterState(Life Cycle Mode=::NewEnumerator1, Loop Behavior=::NewEnumerator1, UseLoopDelay=false)
   &SpawnRate(SpawnRate=56.2449, Use Spawn Probability=false) {enabled=false}
   &SpawnRate(SpawnRate=32.8142, Use Spawn Probability=false) {module=SpawnRate001}
   /SetVariables(Emitter.SpeedLightningPan:NiagaraFloat=&Multiply_Float(A=-3, B=$User.SpeedCurve)) {enabled=false, isSetParameters=true}
  ]
  particleSpawnScript [
   &InitializeParticle(
    Color=<rgba 0 1.4595 2 1>
    Lifetime=0.5
    Mass Max=1.5
    Mass Mode=::NewEnumerator3
    Ribbon Width=11.7242
    Ribbon Width Mode=::NewEnumerator1
    Sprite Rotation Mode=::NewEnumerator2
    Sprite Size Mode=::NewEnumerator1
    Uniform Sprite Size Max=8
    Uniform Sprite Size Min=4
    UsePositionOffset=false
    Write Material Random=true
   )
   &SkeletalMeshLocation(
    Bone Sampling Mode=::NewEnumerator4
    MaskSpawnGroup=false
    Skeletal Mesh=DataInterface() {propertyValues {
      DefaultMesh=null
      ExcludeBoneName=root
      OptionalComponentClass=null
      PreviewMesh=&Fortnite_M_Avg_Player
      SkinningMode=SkinOnTheFly
      SoftSourceActor=null
      SourceMode=Default
      UvSetIndex=0
      WholeMeshLOD=-1
      bExcludeBone=false
      bReadDeformedGeometry=true
      bRequireCurrentFrameData=true
      ComponentTags []
      FilteredBones [hand_r]
      FilteredSockets []
      MeshUserParameter {parameter {name=None, typeDefHandle {registeredTypeIndex=1}, varData []}}
      SamplingRegions []
     }}
   )
  ]
  particleUpdateScript [
   &ParticleState()
   &InheritSourceMovement(Applied Position Scale=<0.7 0.7 0.7>, ApplyForce=false, ApplyPosition=true, ApplyPositionFalloff=false)
   &SolveForcesAndVelocity(Clamp Velocity=false, Limit Acceleration=false)
   &ScaleRibbonWidth(Ribbon Width Scale=&FloatFromCurve(FloatCurve=DataInterface() {propertyValues {Curve {keys [<at 0 1>,<at 1 0>]}}}))
   &ScaleColor(Scale Mode=2)
   /SetVariables(
    Particles.HardEdgeWidth:NiagaraFloat=&FloatFromCurve(FloatCurve=DataInterface() {propertyValues {Curve {keys [<at 0 0.2203>,<at 0.1403 0>]}}}, Scale Curve=$Particles.RibbonWidth)
    Particles.LightningWidth:NiagaraFloat=&FloatFromCurve(FloatCurve=DataInterface() {propertyValues {Curve {keys [<at 0 0>,<at 0.1395 1>,<at 1 0>]}}}, Scale Curve=20)
   ) {isSetParameters=true}
   &DynamicMaterialParameters(Module.Index0Float0Write=true, Module.Index0Float1Write=true, Module.Index0Float2Write=true, Module.Index0Float3Write=true)
  ]
  rendererClasses [&NiagaraRibbonRendererProperties]
  renderers [
   {rendererClass=&NiagaraRibbonRendererProperties}
   {rendererClass=&NiagaraRibbonRendererProperties, rendererIndex=1}
  ]
 }
 Smokelow {
  emitterUpdateScript [
   &EmitterState()
   &SpawnBurst_Instantaneous(Spawn Count=10, Spawn Group=2, Spawn Time=$System.BurstDelay, Use Spawn Probability=false, UseLoopCountLimit=false) {module=SpawnBurst_Instantaneous002}
   &SpawnBurst_Instantaneous(Spawn Count=30, Spawn Time=$System.BurstDelay, Use Spawn Probability=false, UseLoopCountLimit=false)
   &SpawnBurst_Instantaneous(Spawn Count=24, Spawn Group=1, Spawn Time=$System.BurstDelay, Use Spawn Probability=false, UseLoopCountLimit=false) {module=SpawnBurst_Instantaneous001}
  ]
  particleSpawnScript [
   &PartitionParticles(Number Of Partitions=&MergeHelper_IntSwitch(0=0, 1=0, 2=2, 3=?, 4=?, 5=?, Index=&SpawnGroup(), Num=::NewEnumerator1))
   /SetVariables(Particles.VisibilityTag=&SpawnGroup()) {isSetParameters=true}
   /SetVariables(Particles.SpriteSize=&Multiply_Vector2D_ByFloat(Float=1.5, Vector2D=<1 1>)) {isSetParameters=true}
   &InitializeParticle(
    Color=&MergeHelper_ColorSwitch(
     0=<rgba 0.1458 1.2102 2 1>
     1=<rgba 0.1458 1.2102 2 1>
     2=<rgba 1.4583 12.1018 20 1>
     3=?
     Index=&SpawnGroup()
     Num=::NewEnumerator1
    )
    Lifetime=&MergeHelper_FloatSwitch(
     0=&RandomRangeFloat(Minimum=0.46, Maximum=1.0916)
     1=&RandomRangeFloat(Minimum=0.116, Maximum=0.312)
     2=&RandomRangeFloat(Minimum=0.15, Maximum=0.4)
     3=?
     Index=&SpawnGroup()
     Num=::NewEnumerator1
    )
    Mesh Scale=&Multiply_VectorByFloat(
     Vector=/Niagara/DynamicInputs/Helpers/DistributionCurveVector(
      Minimum:Vector3f=<1 1 0.84>
      Maximum:Vector3f=<6.1149 1.1495 1.2202>
      Distribution Curve:NiagaraDataInterfaceCurve=DataInterface() {propertyValues {Curve {
         defaultValue=0
         keys [
          {tangentMode=RCTM_User, time=0, value=0}
          <at 0.5147 0.0853>
          <at 0.7453 0.2618>
          <at 0.9485 0.0971>
          {tangentMode=RCTM_User, time=0.9603, value=0.9882}
         ]
        }}}
      Axis Variation:NiagaraFloat=0
      Randomness Mode:ENiagaraRandomnessMode=::NewEnumerator0
      Evaluation Type:ENiagaraRandomnessEvaluation=::NewEnumerator0
      Seed:NiagaraInt32=0
      Fixed Random Seed:NiagaraBool=false
     )
     Float=1.5252
    )
    Mesh Scale Mode=::NewEnumerator4
    Sprite Rotation Mode=::NewEnumerator2
    Sprite Size=&MergeHelper_Vector2Switch(
     0=&Multiply_Vector2D_ByFloat(Float=&RandomRangeFloat(Minimum=0.5, Maximum=2), Vector2D=<1 1>)
     1=/Niagara/DynamicInputs/Helpers/DistributionCurveVector2D(
      Minimum:Vector2f=<5.8875 84.5851>
      Maximum:Vector2f=<5 373.5695>
      Distribution Curve:NiagaraDataInterfaceCurve=DataInterface() {propertyValues {Curve {defaultValue=0, keys [{tangentMode=RCTM_User, time=0, value=0}, <at 0.8837 0.1088>, {tangentMode=RCTM_User, time=0.8983, value=0.9294}]}}}
      Axis Variation:NiagaraFloat=1
      Randomness Mode:ENiagaraRandomnessMode=::NewEnumerator0
      Evaluation Type:ENiagaraRandomnessEvaluation=::NewEnumerator0
      Seed:NiagaraInt32=0
      Fixed Random Seed:NiagaraBool=false
     )
     2=?
     3=?
     4=?
     Index=&SpawnGroup()
     Num=::NewEnumerator0
    )
    Sprite Size Mode=::NewEnumerator4
    UsePositionOffset=false
    Write Material Random=true
   )
   &CameraOffset(Camera Offset Amount=&RandomRangeFloat(Minimum=0, Maximum=20))
   &ShapeLocation(
    Handle Radius=0
    Large Radius=4.8321
    MaskSpawnGroup=true
    Offset=<0 20 10>
    Offset Mode=::NewEnumerator0
    Override Seed=false
    Rotation Coordinate Space=::World
    Shape Origin=&MakeVector(X=0, Y=0, Z=0)
    Shape Primitive=::NewEnumerator3
    Spawn Group Mask=1
    Surface Distribution=1
    UseOwnerScale=true
   )
   &ShapeLocation(
    MaskSpawnGroup=true
    Offset=<0 20 30>
    Offset Mode=::NewEnumerator0
    Override Seed=false
    Rotation Coordinate Space=::World
    Shape Origin=&MakeVector(X=0, Y=0, Z=0)
    Sphere Radius=5.2565
    UseOwnerScale=true
   ) {module=ShapeLocation001}
   &ShapeLocation(
    Hemisphere Distribution=<1 0.204>
    MaskSpawnGroup=true
    Override Seed=false
    Rotation Angle Type:ENiagara_AngleInput=::NewEnumerator0
    Rotation Mode=::NewEnumerator2
    Spawn Group Mask=2
    Sphere Radius=1
    UseOwnerScale=true
    Yaw / Pitch / Roll=<0 180 0>
   ) {module=ShapeLocation002}
   &AddVelocityFromPoint(
    Module.Use Falloff Distance=false
    Origin Offset=<0 20 0>
    Velocity Origin:Vector3f=&MakeVector(X=0, Y=0, Z=0)
    Velocity Strength=&MergeHelper_FloatSwitch(
     0=&RandomRangeFloat(Minimum=548.7449, Maximum=1884.0087)
     1=0.2312
     2=0
     3=0
     4=0
     Index=&SpawnGroup()
     Num=::NewEnumerator3
    )
   )
   /SetVariables(Particles.Random=&RandomRangeFloat(Minimum=0, Maximum=1)) {isSetParameters=true}
   &InitialMeshOrientation(
    Mesh Orientation Mode=::NewEnumerator2
    Orientation Vector=$Particles.ShapeLocation002.ShapeNormal
    Rotation=&RandomRangeVector(Minimum=<0 0 0>, Maximum=<1 0 0>)
    Use Rotation Vector=true
   ) {module=InitialMeshOrientation002}
  ]
  particleUpdateScript [
   &ParticleState(DeltaTime:NiagaraFloat=&Multiply_Float(A=$Engine.DeltaTime, B=$User.SpeedCurve))
   &ScaleMeshSize(Scale Factor=&Multiply_VectorByFloat(Vector=<1 1 1>, Float=0.6))
   &Drag(
    Drag=&FloatFromCurve(
     FloatCurve=DataInterface() {propertyValues {Curve {
        defaultValue=1
        keys [
         {arriveTangent=4.1785, interpMode=RCIM_Cubic, leaveTangent=4.1785, tangentMode=RCTM_User, time=0, value=0}
         {arriveTangent=1.1179, interpMode=RCIM_Cubic, leaveTangent=1.1179, tangentMode=RCTM_User, time=0.2042, value=0.6228}
         {arriveTangent=0.0224, interpMode=RCIM_Cubic, leaveTangent=0.0224, tangentMode=RCTM_User, time=1, value=1}
        ]
       }}}
     Scale Curve=18.7664
    )
    Use Linear Drag=true
    Use Rotational Drag=true
   )
   &SpriteRotationRate(Rotation Rate=&RandomRangeFloat(Minimum=-200, Maximum=200))
   &ScaleColor(
    Scale Alpha=&FloatFromCurve(
     FloatCurve=DataInterface() {propertyValues {Curve {
        defaultValue=1
        keys [
         {interpMode=RCIM_Cubic, tangentMode=RCTM_User, time=0, value=0}
         {arriveTangent=-2.8196, interpMode=RCIM_Cubic, leaveTangent=-2.8196, tangentMode=RCTM_User, time=0.5, value=1}
         {tangentMode=RCTM_User, time=1, value=0}
        ]
       }}}
     Scale Curve=2
    )
    Scale RGB=&MakeVectorFromLinearColorRGB(Color=&RandomRangeLinearColor(
      Minimum=&ColorFromCurve(ColorCurve=DataInterface() {propertyValues {AlphaCurve {keys []}, BlueCurve {keys [<at 0.6529 1>]}, GreenCurve {keys [<at 0.6529 1>]}, RedCurve {keys [<at 0.6529 1>]}}})
      Maximum=<rgba 1 1 1 1>
     ))
    ScaleA=true
    ScaleRGB=true
   )
   &ScaleSpriteSize(Scale Factor=&Vector2DFromCurve(
     Vector2Curve=DataInterface() {propertyValues {
       XCurve {keys [<at 0 1>,<at 1 1>]}
       YCurve {
        defaultValue=1
        keys [
         {arriveTangent=4.1785, interpMode=RCIM_Cubic, leaveTangent=4.1785, tangentMode=RCTM_User, time=0, value=0}
         {arriveTangent=1.1179, interpMode=RCIM_Cubic, leaveTangent=1.1179, tangentMode=RCTM_User, time=0.2042, value=0.6228}
         {arriveTangent=0.0224, interpMode=RCIM_Cubic, leaveTangent=0.0224, tangentMode=RCTM_User, time=1, value=1}
        ]
       }
      }}
     ScaleCurve=<1 1>
    ))
   &GravityForce(Gravity=&MergeHelper_VectorSwitch(0=<0 0 146.4559>, 1=<0 0 0>, 2=<0 0 0>, 3=<0 0 0>, 4=<0 0 0>, Index=&SpawnGroup(), Num=::NewEnumerator3))
   &CurlNoiseForce(
    Bias Noise=false
    Noise Strength=&MergeHelper_FloatSwitch(
     0=&FloatFromCurve(
      FloatCurve=DataInterface() {propertyValues {Curve {
         defaultValue=1
         keys [
          {arriveTangent=4.1785, interpMode=RCIM_Cubic, leaveTangent=4.1785, tangentMode=RCTM_User, time=0, value=0}
          {arriveTangent=1.1179, interpMode=RCIM_Cubic, leaveTangent=1.1179, tangentMode=RCTM_User, time=0.2042, value=0.6228}
          {arriveTangent=0.0224, interpMode=RCIM_Cubic, leaveTangent=0.0224, tangentMode=RCTM_User, time=1, value=1}
         ]
        }}}
      Scale Curve=500
     )
     1=0
     2=0
     3=0
     4=0
     Index=&SpawnGroup()
     Num=::NewEnumerator3
    )
    Pan Noise=false
   )
   &InheritSourceMovement(ApplyForce=false, ApplyPosition=true, ApplyPositionFalloff=false)
   &SolveForcesAndVelocity(
    Clamp Velocity=false
    DeltaTime:NiagaraFloat=&Multiply_Float(A=$Engine.DeltaTime, B=$User.SpeedCurve)
    Limit Acceleration=false
   )
   &ScaleSpriteSizeBySpeed(Max Scale Factor=&MergeHelper_Vector2Switch(0=<1 3.55>, 1=<0 0>, 2=<0 0>, 3=<0 0>, 4=<0 0>, Index=&SpawnGroup(), Num=::NewEnumerator3))
   &ScaleColor(
    Scale Alpha=&FloatFromCurve(FloatCurve=DataInterface() {propertyValues {Curve {keys [<at 0 1>,<at 1 0>]}}})
    Scale RGB=&MakeVectorFromLinearColorRGB(Color=&MakeCustomLinearColorFromBool(
      Bool=$User.bIsFrontEndPreview
      True Color=&RandomRangeLinearColor(Minimum=<rgba 0.4 0.55 1 1>, Maximum=<rgba 0.08 0.11 0.2 1>)
      False Color=<rgba 1 1 1 1>
     ))
    ScaleA=true
    ScaleRGB=true
   ) {module=ScaleColor001}
   &DynamicMaterialParameters(
    Index 0 Param 1=&Add_Float(
     A=&RandomRangeFloat(Minimum=0, Maximum=1)
     B=&FloatFromCurve(
      FloatCurve=DataInterface() {propertyValues {Curve {keys [
          {arriveTangent=2.7533, interpMode=RCIM_Cubic, leaveTangent=1.8958, tangentMode=RCTM_User, time=0, value=1}
          {arriveTangent=0.8292, interpMode=RCIM_Cubic, leaveTangent=0.8292, tangentMode=RCTM_User, time=0.321, value=1.3453}
          {arriveTangent=0.0529, interpMode=RCIM_Cubic, leaveTangent=0.2021, tangentMode=RCTM_User, time=1, value=1.5}
         ]}}}
      CurveIndex=&Add_Float(A=&Multiply_Float(A=$Particles.Initial.MaterialRandom, B=0.5), B=$Particles.NormalizedAge)
     )
    )
    Index 0 Param 2=&FloatFromCurve(FloatCurve=DataInterface() {propertyValues {Curve {keys [
         {arriveTangent=0.0453, interpMode=RCIM_Cubic, tangentMode=RCTM_Break, time=0, value=-1}
         {arriveTangent=0.0604, interpMode=RCIM_Cubic, leaveTangent=0.0604, tangentMode=RCTM_User, time=1, value=-1}
        ]}}})
    Index 0 Param 3=&FloatFromCurve(
     FloatCurve=DataInterface() {propertyValues {Curve {keys [
         {arriveTangent=0.0632, interpMode=RCIM_Cubic, leaveTangent=0.0632, tangentMode=RCTM_User, time=0, value=20}
         {
          arriveTangent=-1.6858
          arriveTangentWeight=0.0333
          interpMode=RCIM_Cubic
          leaveTangent=-1.6858
          leaveTangentWeight=0.3137
          tangentMode=RCTM_User
          time=0.05
          value=0.1
         }
         {arriveTangentWeight=0.1027, interpMode=RCIM_Cubic, tangentMode=RCTM_Break, time=0.2, value=0}
        ]}}}
     Scale Curve=&RandomRangeFloat(Minimum=0.25, Maximum=0.5)
    )
    Index 0 Param 4=&FloatFromCurve(FloatCurve=DataInterface() {propertyValues {Curve {keys [<at 0.5 1>,<at 1 0>]}}})
    Index 1 Param 1=&FloatFromCurve(
     FloatCurve=DataInterface() {propertyValues {Curve {keys [
         {interpMode=RCIM_Cubic, tangentMode=RCTM_Break, time=0, value=1}
         {arriveTangentWeight=0.0833, interpMode=RCIM_Cubic, tangentMode=RCTM_Break, time=0.25, value=0.3}
        ]}}}
     Scale Curve=&RandomRangeFloat(Minimum=1.5, Maximum=3)
    )
    Index 1 Param 2=&FloatFromCurve(
     FloatCurve=DataInterface() {propertyValues {Curve {keys [
         {interpMode=RCIM_Cubic, leaveTangent=-2.5489, tangentMode=RCTM_Break, time=0, value=1}
         {interpMode=RCIM_Cubic, leaveTangent=0.9066, tangentMode=RCTM_Break, time=1, value=0}
        ]}}}
     Scale Curve=&RandomRangeFloat(Minimum=0.01, Maximum=0.1)
    )
    Index 1 Param 3=&FloatFromCurve(FloatCurve=DataInterface() {propertyValues {Curve {keys [
         {interpMode=RCIM_Cubic, leaveTangent=6.3917, tangentMode=RCTM_Break, time=0, value=0.2}
         {arriveTangent=1.7549, interpMode=RCIM_Cubic, leaveTangent=10.6767, tangentMode=RCTM_User, time=0.25, value=1}
        ]}}})
    Index 1 Param 4=&FloatFromCurve(
     FloatCurve=DataInterface() {propertyValues {Curve {keys [{arriveTangent=-0.136, interpMode=RCIM_Cubic, leaveTangent=-0.136, tangentMode=RCTM_User, time=0, value=1}, <at 1 0.95>]}}}
     Scale Curve=&RandomRangeFloat(Minimum=1.3, Maximum=2)
    )
    Index 2 Param 1=&FloatFromCurve(FloatCurve=DataInterface() {propertyValues {Curve {keys [
         {interpMode=RCIM_Cubic, leaveTangent=4.2485, tangentMode=RCTM_Break, time=0, value=0}
         {arriveTangent=1.0548, interpMode=RCIM_Cubic, leaveTangent=1.0548, tangentMode=RCTM_User, time=0.17, value=0.5715}
         <at 1 1>
        ]}}})
    Index 2 Param 2=&FloatFromCurve(
     FloatCurve=DataInterface() {propertyValues {Curve {keys [
         {arriveTangent=1.9056, interpMode=RCIM_Cubic, leaveTangent=1.9056, tangentMode=RCTM_User, time=0, value=0.2}
         {arriveTangent=0.0259, interpMode=RCIM_Cubic, leaveTangent=0.0259, tangentMode=RCTM_User, time=1, value=1.4}
        ]}}}
     Scale Curve=&RandomRangeFloat(Minimum=0.1, Maximum=0.2)
    )
    Index 2 Param 3=&FloatFromCurve(
     FloatCurve=DataInterface() {propertyValues {Curve {keys [
         {interpMode=RCIM_Cubic, leaveTangent=0.0393, tangentMode=RCTM_Break, time=0, value=1}
         {interpMode=RCIM_Cubic, leaveTangent=11.1233, tangentMode=RCTM_Break, time=1, value=1.5}
        ]}}}
     Scale Curve=0.7
    )
    Index 2 Param 4=&FloatFromCurve(
     FloatCurve=DataInterface() {propertyValues {Curve {keys [
         {arriveTangent=-0.6867, interpMode=RCIM_Cubic, leaveTangent=-0.6867, tangentMode=RCTM_User, time=0, value=1}
         {interpMode=RCIM_Cubic, leaveTangent=-1.2956, tangentMode=RCTM_Break, time=1, value=0.75}
        ]}}}
     Scale Curve=1.5
    )
    Module.Param0WriteEnabled=true
    Module.Param0WriteEnabled1=true
    Module.Param0WriteEnabled2=true
    Module.Param1WriteEnabled=false
    Module.Param1WriteEnabled1=true
    Module.Param1WriteEnabled2=true
    Module.Param2WriteEnabled=true
    Module.Param2WriteEnabled1=true
    Module.Param2WriteEnabled2=true
    Module.Param3WriteEnabled=true
    Module.Param3WriteEnabled1=true
    Module.Param3WriteEnabled2=true
    Write Parameter Index 1=true
    Write Parameter Index 2=true
   ) {module=DynamicMaterialParameters001}
   /SetVariables(Particles.Lightning glow color:LinearColor=&MakeLinearColorFromVectorAndFloat(
     "Vector (RGB)"=&MakeVectorFromLinearColorRGB(Color=&MultiplyLinearColor(
       Linear Color=<rgba 0 0.2223 2 1>
       Scale Factor=&MakeVector4_FromFloat(Float=&RandomRangeFloat(Minimum=0.5, Maximum=1, Evaluation Type=::NewEnumerator1))
      ))
     "Float (Alpha)"=&Multiply_Float(
      A=&FloatFromCurve(FloatCurve=DataInterface() {propertyValues {Curve {keys [<at 0 1 RCIM_Cubic>, {interpMode=RCIM_Cubic, leaveTangent=-3.6307, time=0.9, value=0}]}}})
      B=7
     )
    )) {isSetParameters=true}
   /SetVariables(Particles.Position2:NiagaraPosition=&AddVectorToPosition(Position=$Engine.Emitter.SimulationPosition, Vector=Linked() {linkedVariable {name=Particles.Position, type=NiagaraPosition}})) {isSetParameters=true}
   /SetVariables(
    Particles.DynmParam1:Vector4f=&MakeVector4(
     X=&FloatFromCurve(FloatCurve=DataInterface() {propertyValues {Curve {keys [{tangentMode=RCTM_User, time=0, value=1}, {tangentMode=RCTM_User, time=1, value=1}]}}})
     Y=0.7
     Z=&FloatFromCurve(
      FloatCurve=DataInterface() {propertyValues {Curve {keys [
          {arriveTangent=4.0072, interpMode=RCIM_Cubic, leaveTangent=4.0072, tangentMode=RCTM_User, time=0, value=0}
          {arriveTangent=5, interpMode=RCIM_Cubic, leaveTangent=5, time=0.2543, value=0.6225}
          {arriveTangent=3.1884, interpMode=RCIM_Cubic, leaveTangent=3.1884, time=0.4, value=2}
          {arriveTangent=1.6993, interpMode=RCIM_Cubic, leaveTangent=1.6993, tangentMode=RCTM_User, time=1, value=3}
         ]}}}
      Scale Curve=0.3255
     )
     W=5
    )
    Particles.DynmParam2:Vector4f=&MakeVector4(X=&Lerp_Float(A=1, B=2, Alpha=$Particles.Random), Y=$Particles.Random, Z=2, W=1)
    Particles.DynmParam3:Vector4f=&MakeVector4(
     X=&FloatFromCurve(FloatCurve=DataInterface() {propertyValues {Curve {keys [<at 0 1>,<at 1 0.3>]}}})
     Y=0
     Z=0
     W=0
    )
    Particles.Glow Dy02:Vector4f=&MakeVector4(
     X=&FloatFromCurve(FloatCurve=DataInterface() {propertyValues {Curve {keys [{tangentMode=RCTM_User, time=0, value=0.2}, {tangentMode=RCTM_User, time=0.5, value=1}]}}})
     Y=0.7
     Z=0
     W=5
    )
   ) {isSetParameters=true}
   /SetVariables(Particles.Glow Dy03:Vector4f=&MakeVector4(
     X=&FloatFromCurve(
      FloatCurve=DataInterface() {propertyValues {Curve {keys [<at 0 1>,<at 1 0>]}}}
      Scale Curve=&Lerp_Float(A=2, B=2, Alpha=$Particles.Random)
     )
     Y=$Particles.Random
     Z=2
     W=1
    )) {isSetParameters=true}
  ]
  rendererClasses [&NiagaraSpriteRendererProperties, &NiagaraMeshRendererProperties]
  renderers [
   {rendererClass=&NiagaraSpriteRendererProperties}
   {rendererClass=&NiagaraSpriteRendererProperties, rendererIndex=1}
   {rendererClass=&NiagaraMeshRendererProperties, rendererIndex=2}
   {rendererClass=&NiagaraMeshRendererProperties, rendererIndex=3}
  ]
 }
 WindBurstLWE001 {rendererClasses [], renderers []}
 WindCircularLWE {rendererClasses [], renderers []}
 WindCircularLWE001 {rendererClasses [], renderers []}
}
summary {
 systemName=NS_Emote_HeroicTale_PunchMagic
 emitters [
  {emitterName=HandFX, rendererClasses [&NiagaraSpriteRendererProperties]}
  {emitterName=HandRibbon_, rendererClasses [&NiagaraRibbonRendererProperties]}
  {
   emitterName=Flare002
   rendererClasses [
    &NiagaraSpriteRendererProperties
    &NiagaraMeshRendererProperties
    &NiagaraLightRendererProperties
   ]
  }
  {emitterName=WindCircularLWE, rendererClasses []}
  {emitterName=Smokelow, rendererClasses [&NiagaraSpriteRendererProperties, &NiagaraMeshRendererProperties]}
  {emitterName=WindBurstLWE001, rendererClasses []}
  {emitterName=WindCircularLWE001, rendererClasses []}
 ]
 userVariables [
  User.bIsFrontEnd {type=NiagaraBool}
  User.bIsFrontEndPreview {type=NiagaraBool}
  User.SpeedCurve {type=NiagaraFloat, defaultValue=1, description=""}
 ]
}
